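; Invulnerable brazier switch, usable once by a player carrying a key item, that fires linked Temple Tower fire, vanishing floor and secret door actors.
; Interaction checks KeyItemDefinition in inventory but removes Item Definition, then triggers those actors.
; Authority sets ButtonIsPressed, whose OnRep swaps idle/activate FX and shows light and mask; remotes play interact audio.
; An optional debug flag tosses a pickup of Item Definition on BeginPlay and hides light and mask.
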(event EventBeginPlay
  ; Init/Debug Spawn
  (Variables|BuildingActor|SetbIsInvulnerable true)
  (Variables|Actor|SetbCanBeDamaged false)
  (if (Variables|Default|GetDebug%20Spawn%20Key)
    (Utilities|FlowControl|Delay 1.0)
    (Inventory|SpawnPickupInWorldWithLevel (Variables|Default|GetItem%20Definition) 0 1 (Transformation|GetWorldLocation (Variables|Default|GetArrow)) (* (Transformation|GetForwardVector (Variables|Default|GetArrow)) (Math|Conversions|ToVector(Float) (Variables|Default|GetToss%20Velocity))))
    ; Hide Lights/FX/Wood
    (Rendering|SetVisibility (Variables|Default|GetPointLight) false false)
    (Rendering|SetVisibility (Variables|Default|GetSM_Mask) false false)))

(event Interaction|EventBlueprintOnInteract
  (params (InteractingPawn :type Fort%20Pawn%20Object%20Reference) (InteractionBeingAttempted :type EInteractionBeingAttempted%20Enum :default "FirstInteraction"))
  ; Button Logic
  (Utilities|FlowControl|DoOnce false
    (:Completed
      (Variables|Default|SetButton%20Can%20Be%20Pressed false)
      (CallFunction|RemoveKeyItem InteractingPawn)
      (CallFunction|ButtonPressedEvent)
      (Class|BPTempleTowerFX|ActivateTempleTowerBrazier (Variables|Default|GetTemple%20Tower%20Fire%20Ref))
      (Class|BPHiddenTempleFloor|ActivateTempleVault (Variables|Default|GetVanishing%20Floor%20Ref) 0)
      (Class|BPSecretDoorPressurePlate|OpenVault (Variables|Default|GetSecret%20Door%20Ref) InteractingPawn))))

(event Custom|ButtonPressedEvent
  ; Button Interact
  (Networking|SwitchHasAuthority
    (:Authority
      (Variables|Default|SetButtonIsPressed true))
    (:Remote
      (Audio|Components|Audio|Play (Variables|Default|GetAudio%20Interact) 0.0))))

(fn BlueprintCanInteract
  (params (InteractingPawn :type Fort%20Pawn%20Object%20Reference) (InteractionBeingAttempted :type EInteractionBeingAttempted%20Enum :default "FirstInteraction") (InteractionType :type TInteractionType%20Enum :default "IT_NoInteraction"))
  (bind _asfort_player_pawn_athena (Utilities|Casting|CastToFortPlayerPawnAthena InteractingPawn))
  (bind _asfort_player_controller_athena (Utilities|Casting|CastToFortPlayerControllerAthena (Pawn|GetController _asfort_player_pawn_athena)))
  (return (and (Variables|Default|GetButton%20Can%20Be%20Pressed) (>= (Inventory|GetItemQuantityonInventoryOwner (Fortnite|GetFortPlayerControllerfromActor _asfort_player_controller_athena) (Variables|Default|GetKeyItemDefinition)) 1))))

(fn BlueprintGetInteractionString
  (params (InteractingPawn :type Fort%20Pawn%20Object%20Reference) (InteractionBeingAttempted :type EInteractionBeingAttempted%20Enum :default "FirstInteraction"))
  (return (Variables|Default|GetButtonInteractionText)))

(fn IsPlayer%28s%29inVolume
  (params (Players :type Array%20of%20Actor%20Object%20References))
  (bind _array_element (Utilities|Array|ForEachLoopwithBreak Players)
    (:LoopBody
      (bind _asfort_player_pawn_athena (Utilities|Casting|CastToFortPlayerPawnAthena _array_element))
      (Utilities|Casting|CastToFortPlayerControllerAthena (Pawn|GetController _asfort_player_pawn_athena))
      (LocalVariables|SetIs%20in%20Volume true))
    (:Completed
      _))
  (return (LocalVariables|GetIs%20in%20Volume)))

(fn OnRep_ButtonIsPressed
  (bind _pointlight (Variables|Default|GetPointLight))
  (bind _sm_mask (Variables|Default|GetSM_Mask))
  (if (Variables|Default|GetButtonIsPressed)
    ; Enable Point Light
    (Rendering|SetVisibility _pointlight true false)
    ; Enable Mask
    (Rendering|SetVisibility _sm_mask true false)
    (Components|Activation|Activate (Variables|Default|GetFX_Activate) false)
    (Components|Activation|Deactivate (Variables|Default|GetFX_Idle))
    (else
      (Rendering|SetVisibility _pointlight false false)
      (Rendering|SetVisibility _sm_mask false false))))

(fn OnRep_IsInsideTrigger)

(fn RemoveKeyItem
  (params (Pawn :type Fort%20Pawn%20Object%20Reference))
  (bind _asfort_player_controller_athena (Utilities|Casting|CastToFortPlayerControllerAthena (Pawn|GetController Pawn)))
  (Inventory|RemoveItemfromInventoryOwner _asfort_player_controller_athena (Variables|Default|GetItem%20Definition)))

(fn ConstructionScript)
