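; BP_FloatingObject is an invulnerable prototype level-design actor whose cube drifts and tumbles near its spawn point.
; On BeginPlay it records the root's world location, enables physics without gravity, then applies a mass-scaled random impulse and random torque.
; A looping one-second timer checks the cube's distance from that start and, beyond a fixed radius, applies a force back toward it.
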(event EventBeginPlay
  (bind _dynamic_cube_87f1fb2d (Variables|Default|GetDynamic_Cube_87F1FB2D))
  (Variables|BuildingActor|SetbIsInvulnerable true)
  (Variables|Actor|SetbCanBeDamaged false)
  (Utilities|FlowControl|Sequence
    (:then_0
      ; Setting Initial Location of Object
      (Variables|Default|SetStarting%20Location (Transformation|GetWorldLocation (Variables|Default|GetDefaultSceneRoot)))
      ; Set Physics/Impulse/Rotation
      (Physics|SetSimulatePhysics _dynamic_cube_87f1fb2d true)
      (Physics|SetEnableGravity _dynamic_cube_87f1fb2d false)
      (Physics|AddImpulse _dynamic_cube_87f1fb2d (* (* (Math|Random|RandomUnitVector) (Variables|Default|GetImpulse)) (Physics|GetMass (Variables|Default|GetDynamic_Cube_87F1FB2D))) "None" false)
      (Physics|AddTorqueinRadians (Variables|Default|GetDynamic_Cube_87F1FB2D) (* (Math|Random|RandomUnitVector) (Variables|Default|GetRotation)) "None" false))
    (:then_1
      (Utilities|Time|SetTimerbyEvent (delegate Custom|CustomEvent) 1.0 true))))

(event Custom|CustomEvent
  (bind _dynamic_cube_87f1fb2d (Variables|Default|GetDynamic_Cube_87F1FB2D))
  (bind _returnvalue (Transformation|GetWorldLocation _dynamic_cube_87f1fb2d))
  (bind _starting_location (Variables|Default|GetStarting%20Location))
  ; if distance is greater than value, send back to starting location
  (if (> (Math|Vector|Distance(Vector) _starting_location _returnvalue) 5e+02)
    ; Amount of Return Force
    (Physics|AddForce (Variables|Default|GetDynamic_Cube_87F1FB2D) (* (Math|Vector|GetUnitDirection(Vector) _returnvalue _starting_location) (Variables|Default|GetForce%20%28Return%29)) "None" false)))

(fn ConstructionScript)
