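; Prototype Geno cable actor that responds to BP_GenoManager console sequences by writing start times into its Cable component's custom primitive data.
; BeginPlay finds the manager, binds its sequence and reset dispatchers, and replays matching console event history, or resets if none exists.
; Server world time and per-bundle offsets determine whether each write is delayed or backdated; reset zeroes eight data slots.
; The construction script syncs the CableBundles map size to CableBundleCount.
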(event EventBeginPlay
  (Utilities|FlowControl|Sequence
    (:then_0
      (Utilities|IsValid (Variables|CableSettings|GetManager)
        (:Is%20Valid
          _)
        (:Is%20Not%20Valid
          (bind _outactors (Actor|GetAllActorsOfClass "/Hera/Prototype/LevelDesign/Allen/BP_GenoManager.BP_GenoManager_C"))
          (if (Utilities|Array|IsNotEmpty _outactors)
            (Variables|CableSettings|SetManager (Utilities|Array|Get(acopy) _outactors)))))
      (Default|BindEventtoConsoleSequence (Variables|CableSettings|GetManager) (EventDispatchers|CreateEvent :!event Custom|BeginConsoleSequence))
      (Default|BindEventtoResetConsoleStates (Variables|CableSettings|GetManager) (EventDispatchers|CreateEvent :!event Custom|ResetCableState)))
    (:then_1
      (bind _manager (Variables|CableSettings|GetManager)
        (:then
          (bind _consoleeventhistory (Class|BPGenoManager|GetConsoleEventHistory _manager))
          (if (Utilities|Array|IsNotEmpty _consoleeventhistory)
            (bind _keys (Utilities|Map|Keys (Variables|CableSettings|GetCableBundles)))
            (for _array_element _keys
              (bind _array_element_2 (Utilities|Array|ForEachLoopwithBreak _consoleeventhistory)
                (:LoopBody
                  (if (== (Utilities|Struct|BreakSGenoManagerEventList _array_element_2) _array_element)
                    (CallFunction|BeginConsoleSequence (Utilities|Struct|BreakSGenoManagerEventList _array_element_2) (.EventStartTime_5_4B005E504ABB04C687DA71A38EE220D5 (Utilities|Struct|BreakSGenoManagerEventList _array_element_2)))))))
            (else
              (CallFunction|ResetCableState))))
        (:else)))))

(event Custom|BeginConsoleSequence
  (params (SequenceIndex :type Integer) (StartTime :type Float%20%28double-precision%29))
  (bind _cablebundles (Variables|CableSettings|GetCableBundles))
  (bind _value (Utilities|Map|Find _cablebundles SequenceIndex))
  (bind _starttimeoffset_4_27dbc86f4d16262d46650fa452f329e5 (Utilities|Struct|BreakSGenoCableData _value))
  (bind _returnvalue (Fortnite|Sync|Time|GetServerWorldTimeSeconds))
  (bind _returnvalue_2 (- StartTime _returnvalue))
  (bind _returnvalue_3 (- _starttimeoffset_4_27dbc86f4d16262d46650fa452f329e5 _returnvalue_2))
  (bind _keys (Utilities|Map|Keys (Variables|CableSettings|GetCableBundles)))
  (if (Utilities|Array|ContainsItem _keys SequenceIndex)
    (if (>= _returnvalue_3 0)
      (Utilities|FlowControl|Delay _returnvalue_3)
      (CallFunction|BeginPassPrimitiveData SequenceIndex (Utilities|Time|GetGameTimeinSeconds))
      (else
        (CallFunction|BeginPassPrimitiveData SequenceIndex (+ (Utilities|Time|GetGameTimeinSeconds) _returnvalue_3))))))

(event Custom|BeginPassPrimitiveData
  (params (BundleIndex :type Integer) (Time :type Float%20%28double-precision%29))
  (bind _cablebundles (Variables|CableSettings|GetCableBundles))
  (bind _value (Utilities|Map|Find _cablebundles BundleIndex))
  (bind (_cablelength_13_e37bacf849c9f1178b16cabc568ecdef _bundlematerialidremap_8_230742a043eaae7e5c748781025dd443) (Utilities|Struct|BreakSGenoCableData _value))
  (Rendering|Material|SetCustomPrimitiveDataFloat (Variables|Default|GetCable) _bundlematerialidremap_8_230742a043eaae7e5c748781025dd443 Time)
  (Rendering|Material|SetDefaultCustomPrimitiveDataFloat (Variables|Default|GetCable) _bundlematerialidremap_8_230742a043eaae7e5c748781025dd443 Time))

(event Custom|ResetCableState
  (CallFunction|ResetCableTime false))

(fn ResetCableTime
  (params (SingleIndex :type Boolean) (Index :type Integer))
  (if SingleIndex
    (Rendering|Material|SetCustomPrimitiveDataFloat (Variables|Default|GetCable) Index 0.0)
    (Rendering|Material|SetDefaultCustomPrimitiveDataFloat (Variables|Default|GetCable) Index 0.0)
    (else
      (for _index (range 8)
        (Rendering|Material|SetCustomPrimitiveDataFloat (Variables|Default|GetCable) _index 0.0)
        (Rendering|Material|SetDefaultCustomPrimitiveDataFloat (Variables|Default|GetCable) _index 0.0)))))

(fn ConstructionScript
  (bind _cablebundles (Variables|CableSettings|GetCableBundles))
  (bind _returnvalue (Utilities|Map|Length _cablebundles))
  (bind _cablebundlecount (Variables|CableSettings|GetCableBundleCount))
  (bind _returnvalue_2 (- _returnvalue _cablebundlecount))
  (if (!= _returnvalue_2 0)
    (LocalVariables|SetCableMapDifference _returnvalue_2)
    (if (> _returnvalue_2 0)
      (for _ (range (LocalVariables|GetCableMapDifference))
        (bind _keys (Utilities|Map|Keys (Variables|CableSettings|GetCableBundles)))
        (LocalVariables|SetCableKeys _keys)
        (Utilities|Map|Remove (Variables|CableSettings|GetCableBundles) (Utilities|Array|LastIndex (LocalVariables|GetCableKeys))))
      (else
        (for _ (range (Math|Integer|Absolute(Integer) (LocalVariables|GetCableMapDifference)))
          (LocalVariables|SetKeyAdded false)
          (bind _keys_2 (Utilities|Map|Keys (Variables|CableSettings|GetCableBundles)))
          (LocalVariables|SetCableKeys _keys_2)
          (bind (_array_element _array_index) (Utilities|Array|ForEachLoopwithBreak (LocalVariables|GetCableKeys))
            (:LoopBody
              (if (not (Utilities|Array|ContainsItem (LocalVariables|GetCableKeys) _array_index))
                (Utilities|Map|Add (Variables|CableSettings|GetCableBundles) _array_index (Utilities|Struct|MakeSGenoCableData 0.0))
                (LocalVariables|SetKeyAdded true)))
            (:Completed
              _))
          (if (not (LocalVariables|GetKeyAdded))
            (Utilities|Map|Add (Variables|CableSettings|GetCableBundles) (Utilities|Array|Length (LocalVariables|GetCableKeys)) (Utilities|Struct|MakeSGenoCableData 0.0))))))))
