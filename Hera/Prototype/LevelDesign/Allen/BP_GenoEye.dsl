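; BP_GenoEye is a prototype two-eyed prop whose eyes independently idle-glance, look at locations, or track actors.
; On a GenoManager power-surge event, it waits a cable-bundle offset, triggers piston contraction and looks toward a console location.
; Tracking periodically gives the newest actor to the eye on its side, idling the other.
; Cone-clamped targets, focal distances and slerp timings go into Custom Primitive Data, presumably for material-driven animation.
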(event EventBeginPlay
  (Utilities|FlowControl|Sequence
    (:then_0
      (Utilities|IsValid (Variables|GenoEyeSettings|GetManager)
        (:Is%20Valid
          _)
        (:Is%20Not%20Valid
          (bind _outactors (Actor|GetAllActorsOfClass "/Hera/Prototype/LevelDesign/Allen/BP_GenoManager.BP_GenoManager_C"))
          (if (Utilities|Array|IsNotEmpty _outactors)
            (Variables|GenoEyeSettings|SetManager (Utilities|Array|Get(acopy) _outactors)))))
      (Default|BindEventtoConsoleSequence (Variables|GenoEyeSettings|GetManager) (EventDispatchers|CreateEvent :!event Custom|BeginReactPowerSurge)))
    (:then_1
      (CallFunction|BeginIdleLookAt false)
      (CallFunction|BeginIdleLookAt true))))

(event Custom|BeginIdleLookAt
  (params (EyeB :type Boolean))
  (if EyeB
    (Variables|EyeB|SetCurrentEyeStateB "NewEnumerator0")
    (else
      (Variables|EyeA|SetCurrentEyeStateA "NewEnumerator0")))
  (CallFunction|DoIdleLookAt EyeB))

(event Custom|DoIdleLookAt
  (params (EyeB :type Boolean))
  (Utilities|FlowControl|Sequence
    (:then_0
      (CallFunction|SetEyeLookatTarget (+ (* (Math|Random|RandomUnitVectorinConeinDegrees (select EyeB (Math|Vector|NegateVector (Transformation|GetRightVector (Variables|Default|GetGenoEye))) (Transformation|GetRightVector (Variables|Default|GetGenoEye))) 15.0) (Math|Random|RandomFloatinRange 4e+04 6.5e+04)) (Variables|GenoEyeSettings|GetLookAtPivot)) EyeB))
    (:then_1
      (Variables|Default|SetIdleTimeDelta (Math|Random|RandomFloatinRange 0.3 2.0))
      (if EyeB
        (Utilities|FlowControl|Delay (Variables|Default|GetIdleTimeDelta))
        (if (Utilities|Enum|Equal(Enum) "NewEnumerator0" (Variables|EyeB|GetCurrentEyeStateB))
          (CallFunction|DoIdleLookAt true))
        (else
          (Utilities|FlowControl|Delay (Variables|Default|GetIdleTimeDelta))
          (if (Utilities|Enum|Equal(Enum) "NewEnumerator0" (Variables|EyeA|GetCurrentEyeStateA))
            (CallFunction|DoIdleLookAt false)))))))

(event Custom|BeginPistonContraction
  (bind _returnvalue (Utilities|Time|GetGameTimeinSeconds))
  (Rendering|Material|SetCustomPrimitiveDataFloat (Variables|Default|GetGenoEye) 19 _returnvalue)
  (Rendering|Material|SetDefaultCustomPrimitiveDataFloat (Variables|Default|GetGenoEye) 19 _returnvalue))

(event Custom|BeginLookAtLocation
  (params (Location :type Vector) (Duration :type Float%20%28double-precision%29 :default 1e+01))
  (bind _eyeb (Default|LocationEyeDirection Location))
  (if _eyeb
    (Variables|EyeB|SetCurrentEyeStateB "NewEnumerator1")
    (else
      (Variables|EyeA|SetCurrentEyeStateA "NewEnumerator1")))
  (CallFunction|SetEyeLookatTarget Location _eyeb)
  (if _eyeb
    (Utilities|FlowControl|RetriggerableDelay Duration)
    (if (Utilities|Enum|Equal(Enum) "NewEnumerator1" (Variables|EyeB|GetCurrentEyeStateB))
      (CallFunction|EndState true))
    (else
      (Utilities|FlowControl|RetriggerableDelay Duration)
      (if (Utilities|Enum|Equal(Enum) "NewEnumerator1" (Variables|EyeA|GetCurrentEyeStateA))
        (CallFunction|EndState false)))))

(event Custom|EndState
  (params (EyeB :type Boolean))
  (CallFunction|BeginIdleLookAt EyeB))

(event Custom|BeginReactPowerSurge
  (params (SequenceIndex :type Integer) (StartTime :type Float%20%28double-precision%29))
  (bind _value (Utilities|Map|Find (Variables|GenoEyeSettings|GetCableBundles) SequenceIndex))
  (bind _value_starttimeoffset_4_27dbc86f4d16262d46650fa452f329e5 (Utilities|Struct|BreakSGenoCableData _value))
  (bind _sequenceindex (Variables|Default|GetSequenceIndex))
  (bind _consolelocations (Variables|GenoEyeSettings|GetConsoleLocations))
  (Variables|Default|SetSequenceIndex SequenceIndex)
  (Utilities|FlowControl|Delay _value_starttimeoffset_4_27dbc86f4d16262d46650fa452f329e5)
  (CallFunction|BeginPistonContraction)
  (if (Utilities|Array|IsValidIndex _consolelocations _sequenceindex)
    (CallFunction|BeginLookatLocation (Math|Vector|Normalize (Utilities|Array|Get(acopy) _consolelocations _sequenceindex)) 5.0)
    (else
      (Development|PrintString "Warning: Invalid console index" false))))

(event Custom|BeginTrackingActor
  (params (Actor :type Actor%20Object%20Reference))
  (Utilities|IsValid Actor
    (:Is%20Valid
      (Utilities|Array|AddUnique (Variables|Default|GetTrackedActors) Actor)
      (CallFunction|UpdateRelevantActors))
    (:Is%20Not%20Valid
      (CallFunction|EndTrackingActor Actor))))

(event Custom|UpdateLookAtActorLocation
  (params (EyeB :type Boolean))
  (bind _currenttrackingactorb (Variables|EyeB|GetCurrentTrackingActorB))
  (bind _currenttrackingactora (Variables|EyeA|GetCurrentTrackingActorA))
  (bind _returnvalue (select EyeB _currenttrackingactorb _currenttrackingactora))
  (Utilities|IsValid _returnvalue
    (:Is%20Valid
      (if (Utilities|Enum|Equal(Enum) (select EyeB (Variables|EyeB|GetCurrentEyeStateB) (Variables|EyeA|GetCurrentEyeStateA)) "NewEnumerator2")
        (CallFunction|SetEyeLookatTarget (Transformation|GetActorLocation _returnvalue) EyeB)
        (else
          (CallFunction|EndLookatActor EyeB))))
    (:Is%20Not%20Valid
      (CallFunction|EndLookatActor EyeB))))

(event Custom|EndLookAtActor
  (params (EyeB :type Boolean))
  (CallFunction|EndState EyeB))

(event Custom|UpdateRelevantActors
  (bind _trackedactors (Variables|Default|GetTrackedActors))
  (bind _returnvalue (Utilities|Array|LastIndex _trackedactors))
  (bind _output (Utilities|Array|Get(acopy) _trackedactors _returnvalue))
  (Utilities|FlowControl|Sequence
    (:then_0
      (Variables|EyeA|SetCurrentTrackingActorA 0)
      (Variables|EyeB|SetCurrentTrackingActorB 0))
    (:then_1
      (if (Utilities|Array|IsEmpty (Variables|Default|GetTrackedActors))
        (CallFunction|EndTrackingActors)))
    (:then_2
      (Utilities|IsValid _output
        (:Is%20Valid
          (if (CallFunction|TrackingActorEyeDirection _output)
            (Variables|EyeB|SetCurrentTrackingActorB _output)
            (Variables|EyeB|SetCurrentEyeStateB "NewEnumerator2")
            (CallFunction|UpdateLookatActorLocation true)
            (CallFunction|EndState false)
            (else
              (Variables|EyeA|SetCurrentTrackingActorA _output)
              (Variables|EyeA|SetCurrentEyeStateA "NewEnumerator2")
              (CallFunction|UpdateLookatActorLocation false)
              (CallFunction|EndState true))))
        (:Is%20Not%20Valid
          (CallFunction|EndTrackingActor _output))) ; CURRENTLY: just get most recent actor
    )
    (:then_3
      (Utilities|FlowControl|Delay (Variables|GenoEyeSettings|GetTrackingFrequency))
      (if (Utilities|Array|IsNotEmpty (Variables|Default|GetTrackedActors))
        (CallFunction|UpdateRelevantActors)
        (else
          (CallFunction|EndTrackingActors))))))

(event Custom|EndTrackingActor
  (params (Actor :type Actor%20Object%20Reference))
  (Utilities|FlowControl|Sequence
    (:then_0
      (if (== Actor (Variables|EyeA|GetCurrentTrackingActorA))
        (switch Utilities|FlowControl|Switch|SwitchonE_GenoEyeStates (Variables|EyeA|GetCurrentEyeStateA)
          (:NewEnumerator2
            (CallFunction|EndState false)))))
    (:then_1
      (if (== Actor (Variables|EyeB|GetCurrentTrackingActorB))
        (switch Utilities|FlowControl|Switch|SwitchonE_GenoEyeStates (Variables|EyeB|GetCurrentEyeStateB)
          (:NewEnumerator2
            (CallFunction|EndState true)))))
    (:then_2
      (Utilities|Array|RemoveItem (Variables|Default|GetTrackedActors) Actor)
      (CallFunction|UpdateRelevantActors))))

(event Custom|EndTrackingActors
  (Utilities|Array|Clear (Variables|Default|GetTrackedActors))
  (Utilities|FlowControl|Sequence
    (:then_0
      (switch Utilities|FlowControl|Switch|SwitchonE_GenoEyeStates (Variables|EyeA|GetCurrentEyeStateA)
        (:NewEnumerator2
          (CallFunction|EndState false))))
    (:then_1
      (switch Utilities|FlowControl|Switch|SwitchonE_GenoEyeStates (Variables|EyeB|GetCurrentEyeStateB)
        (:NewEnumerator2
          (CallFunction|EndState true))))))

(fn ClampVectorToCone
  (params (Vector :type Vector) (ConeDirection :type Vector :default "1.000000,0.000000,0.000000") (ConeHalfAngleDegrees :type Float%20%28double-precision%29))
  (LocalVariables|SetL_Vector (Math|Vector|Normalize Vector))
  (LocalVariables|SetL_ConeDirection (Math|Vector|Normalize ConeDirection))
  (if (< (Math|Vector|DotProduct (LocalVariables|GetL_Vector) (LocalVariables|GetL_ConeDirection)) (Math|Trig|Cos(Degrees) (Variables|GetConeHalfAngleDegrees)))
    (LocalVariables|SetL_ClampedVector (+ (* (LocalVariables|GetL_ConeDirection) (Math|Trig|Cos(Degrees) (Variables|GetConeHalfAngleDegrees))) (* (Math|Vector|Normalize (- (LocalVariables|GetL_Vector) (* (LocalVariables|GetL_ConeDirection) (Math|Vector|DotProduct (LocalVariables|GetL_Vector) (LocalVariables|GetL_ConeDirection))))) (Math|Trig|Sin(Degrees) (Variables|GetConeHalfAngleDegrees)))))
    (LocalVariables|SetL_ClampOccured true)
    (else
      (LocalVariables|SetL_ClampedVector (LocalVariables|GetL_Vector))
      (LocalVariables|SetL_ClampOccured false)))
  (return (LocalVariables|GetL_ClampedVector) (LocalVariables|GetL_ClampOccured)))

(fn Gain
  (params (Base :type Float%20%28double-precision%29) (Exponent :type Float%20%28double-precision%29 :default 2.0))
  (return (Math|Float|SelectFloat (- 1.0 (* (Math|Float|Power (* (- 1.0 (Variables|GetBase)) 2.0) (Variables|GetExponent)) 0.5)) (* (Math|Float|Power (* (Variables|GetBase) 2.0) (Variables|GetExponent)) 0.5) (>= (Variables|GetBase) 0.5))))

(fn LocationEyeDirection
  (params (Location :type Vector))
  (return (> (Math|Vector|DotProduct (Math|Vector|Normalize (- (Variables|GenoEyeSettings|GetLookAtPivot) Location)) (Transformation|GetRightVector (Variables|Default|GetGenoEye))) 0)))

(fn SetEyeLookAtTarget
  (params (TargetLocation :type Vector) (EyeB :type Boolean))
  (if EyeB
    (CallFunction|SetLookatTarget (Variables|GetTargetLocation) (Variables|EyeB|GetCurrentTargetB) (Variables|EyeB|GetPreviousTargetB) (Variables|EyeB|GetSlerpDurationB) (Variables|EyeB|GetUpdateTimeB) (Variables|EyeB|GetCurrentFocalDistanceB) (Variables|EyeB|GetPreviousFocalDistanceB) 11 14 17 18 22 23 "0.000000,1.000000,0.000000" (Variables|GenoEyeSettings|GetLookAtMaxConeAngleDegrees))
    (else
      (CallFunction|SetLookatTarget (Variables|GetTargetLocation) (Variables|EyeA|GetCurrentTargetA) (Variables|EyeA|GetPreviousTargetA) (Variables|EyeA|GetSlerpDurationA) (Variables|EyeA|GetUpdateTimeA) (Variables|EyeA|GetCurrentFocalDistanceA) (Variables|EyeA|GetPreviousFocalDistanceA) 3 6 9 10 20 21 "0.000000,-1.000000,0.000000" (Variables|GenoEyeSettings|GetLookAtMaxConeAngleDegrees)))))

(fn SetLookAtTarget
  (params (NewTargetLocation :type Vector) (TargetUnitVector :type Vector%20%28by%20ref%29) (PreviousTargetUnitVector :type Vector%20%28by%20ref%29) (Duration :type Float%20%28double-precision%29%20%28by%20ref%29) (UpdateTime :type Float%20%28double-precision%29%20%28by%20ref%29) (TargetFocalDistance :type Float%20%28double-precision%29%20%28by%20ref%29) (PreviousFocalDistance :type Float%20%28double-precision%29%20%28by%20ref%29) (NewTargetCPDIndex :type Integer) (PreviousTargetCPDIndex :type Integer) (EventTimeCPDIndex :type Integer) (EventDurationCPDIndex :type Integer) (TargetFocalDistanceCPDIndex :type Integer) (PreviousFocalDistanceCPDIndex :type Integer) (ClampConeDirection :type Vector :default "0.000000,1.000000,0.000000") (ClampConeHalfAngle :type Float%20%28double-precision%29))
  (bind _newtargetcpdindex (Variables|GetNewTargetCPDIndex))
  (bind _eventtimecpdindex (Variables|GetEventTimeCPDIndex))
  (bind _targetunitvector (Variables|GetTargetUnitVector))
  (bind _previoustargetunitvector (Variables|GetPreviousTargetUnitVector))
  (bind _previoustargetcpdindex (Variables|GetPreviousTargetCPDIndex))
  (bind _updatetime (Variables|GetUpdateTime))
  (bind _targetfocaldistancecpdindex (Variables|GetTargetFocalDistanceCPDIndex))
  (bind _targetfocaldistance (Variables|GetTargetFocalDistance))
  (bind _previousfocaldistancecpdindex (Variables|GetPreviousFocalDistanceCPDIndex))
  (bind _previousfocaldistance (Variables|GetPreviousFocalDistance))
  (bind _eventdurationcpdindex (Variables|GetEventDurationCPDIndex))
  (bind _duration (Variables|GetDuration))
  (Utilities|FlowControl|Sequence
    (:then_0
      (Utilities|FlowControl|Sequence
        (:then_0
          (if (< (Utilities|Time|GetGameTimeinSeconds) (+ (Variables|GetUpdateTime) (Variables|GetDuration)))
            (LocalVariables|SetL_TransitionPhase (/ (- (Utilities|Time|GetGameTimeinSeconds) (Variables|GetUpdateTime)) (Variables|GetDuration)))
            (Utilities|FlowControl|Sequence
              (:then_0
                (Variables|SetBy-RefVar (Variables|GetPreviousTargetUnitVector) (Math|Fortnite|VectorSlerp (Math|Vector|MakeVector 0.0) (Variables|GetPreviousTargetUnitVector) (Variables|GetTargetUnitVector) (CallFunction|Gain (LocalVariables|GetL_TransitionPhase) 4.0))))
              (:then_1
                (Variables|SetBy-RefVar (Variables|GetPreviousFocalDistance) (Math|Float|Lerp (Variables|GetPreviousFocalDistance) (Variables|GetTargetFocalDistance) (LocalVariables|GetL_TransitionPhase)))))
            (else
              (Utilities|FlowControl|Sequence
                (:then_0
                  (Variables|SetBy-RefVar (Variables|GetPreviousTargetUnitVector) (Math|Vector|Normalize (Variables|GetTargetUnitVector))))
                (:then_1
                  (Variables|SetBy-RefVar (Variables|GetPreviousFocalDistance) (Variables|GetTargetFocalDistance)))))))
        (:then_1
          (LocalVariables|SetL_TargetVector (- (Variables|GenoEyeSettings|GetLookAtPivot) (Variables|GetNewTargetLocation))))
        (:then_2
          (Variables|SetBy-RefVar (Variables|GetTargetFocalDistance) (Math|Vector|VectorLength (LocalVariables|GetL_TargetVector))))
        (:then_3
          (Variables|SetBy-RefVar (Variables|GetTargetUnitVector) (Math|Vector|Normalize (LocalVariables|GetL_TargetVector))))
        (:then_4
          (LocalVariables|SetL_ClampConeDirection (Math|Vector|Normalize (Variables|GetClampConeDirection))))))
    (:then_1
      (if (!= 0 (Variables|GetClampConeHalfAngle))
        (Variables|SetBy-RefVar (Variables|GetTargetUnitVector) (CallFunction|ClampVectortoCone (Variables|GetTargetUnitVector) (LocalVariables|GetL_ClampConeDirection) (Variables|GetClampConeHalfAngle)))
        (Variables|SetBy-RefVar (Variables|GetPreviousTargetUnitVector) (CallFunction|ClampVectortoCone (Variables|GetPreviousTargetUnitVector) (LocalVariables|GetL_ClampConeDirection) (Variables|GetClampConeHalfAngle)))))
    (:then_2
      (Variables|SetBy-RefVar (Variables|GetDuration) (* (CallFunction|SphericalDistance (Variables|GetTargetUnitVector) (Variables|GetPreviousTargetUnitVector)) (Variables|GenoEyeSettings|GetLookAtSpeedRadians)))
      (Variables|SetBy-RefVar (Variables|GetUpdateTime) (Utilities|Time|GetGameTimeinSeconds)))
    (:then_3
      (Utilities|FlowControl|Sequence
        (:then_0
          (Rendering|Material|SetDefaultCustomPrimitiveDataVector3 (Variables|Default|GetGenoEye) _newtargetcpdindex _targetunitvector)
          (Rendering|Material|SetCustomPrimitiveDataVector3 (Variables|Default|GetGenoEye) _newtargetcpdindex _targetunitvector))
        (:then_1
          (Rendering|Material|SetDefaultCustomPrimitiveDataVector3 (Variables|Default|GetGenoEye) _previoustargetcpdindex _previoustargetunitvector)
          (Rendering|Material|SetCustomPrimitiveDataVector3 (Variables|Default|GetGenoEye) _previoustargetcpdindex _previoustargetunitvector))
        (:then_2
          (Rendering|Material|SetDefaultCustomPrimitiveDataFloat (Variables|Default|GetGenoEye) _eventtimecpdindex _updatetime)
          (Rendering|Material|SetDefaultCustomPrimitiveDataFloat (Variables|Default|GetGenoEye) _eventtimecpdindex _updatetime))
        (:then_3
          (Rendering|Material|SetDefaultCustomPrimitiveDataFloat (Variables|Default|GetGenoEye) _eventdurationcpdindex _duration)
          (Rendering|Material|SetDefaultCustomPrimitiveDataFloat (Variables|Default|GetGenoEye) _eventdurationcpdindex _duration))
        (:then_4
          (Rendering|Material|SetDefaultCustomPrimitiveDataFloat (Variables|Default|GetGenoEye) _targetfocaldistancecpdindex _targetfocaldistance)
          (Rendering|Material|SetDefaultCustomPrimitiveDataFloat (Variables|Default|GetGenoEye) _targetfocaldistancecpdindex _targetfocaldistance))
        (:then_5
          (Rendering|Material|SetDefaultCustomPrimitiveDataFloat (Variables|Default|GetGenoEye) _previousfocaldistancecpdindex _previousfocaldistance)
          (Rendering|Material|SetDefaultCustomPrimitiveDataFloat (Variables|Default|GetGenoEye) _previousfocaldistancecpdindex _previousfocaldistance))) ; SetCPD
    )))

(fn SphericalDistance
  (params (A :type Vector) (B :type Vector))
  (return (Math|Float|Absolute(Float) (Math|Trig|Acos(Radians) (Math|Float|Clamp(Float) (Math|Vector|DotProduct (Math|Vector|Normalize A) (Math|Vector|Normalize B)) -1.0)))))

(fn TrackingActorEyeDirection
  (params (Actor :type Actor%20Object%20Reference))
  (return (Default|LocationEyeDirection (Transformation|GetActorLocation Actor))))

(fn ConstructionScript
  (bind _returnvalue (Transformation|GetActorTransform))
  (bind _lookatpivot (Variables|GenoEyeSettings|GetLookAtPivot))
  (bind _genoeye (Variables|Default|GetGenoEye))
  (Utilities|FlowControl|Sequence
    (:then_0
      (Variables|GenoEyeSettings|SetLookAtPivot (Math|Transform|InverseTransformLocation _returnvalue (Transformation|GetSocketLocation (Variables|Default|GetGenoEye) "Pivot")))
      (Rendering|Material|SetDefaultCustomPrimitiveDataVector3 _genoeye 0 _lookatpivot)
      (Rendering|Material|SetCustomPrimitiveDataVector3 _genoeye 0 _lookatpivot))
    (:then_1
      (Variables|GenoEyeSettings|SetLookAtPivot (Math|Transform|TransformLocation _returnvalue (Variables|GenoEyeSettings|GetLookAtPivot))))
    (:then_2
      (Variables|GenoEyeSettings|SetLookAtSpeedRadians (Math|Trig|DegreesToRadians (Variables|GenoEyeSettings|GetLookAtSpeed))))
    (:then_3
      (CallFunction|SetEyeLookatTarget (Math|Transform|TransformLocation (Transformation|GetActorTransform) (Variables|GenoEyeSettings|GetDebugTargetA)))
      (CallFunction|SetEyeLookatTarget (Math|Transform|TransformLocation (Transformation|GetActorTransform) (Variables|GenoEyeSettings|GetDebugTargetB)) true))))
