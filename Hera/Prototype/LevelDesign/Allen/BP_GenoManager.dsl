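; BP_GenoManager is a prototype level-design manager that records and broadcasts console sequence events.
; Beginning a sequence inserts the instigator index and server world time at the front of an event history array, newest first, then calls the ConsoleSequence dispatcher with both values.
; Resetting clears that history and calls the ResetConsoleStates dispatcher.
; A getter function exposes the history array.
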(event Custom|BeginConsoleSequence
  (params (InstigatorIndex :type Integer))
  (bind _returnvalue (Fortnite|Sync|Time|GetServerWorldTimeSeconds))
  (Utilities|Array|Insert (Variables|Default|GetConsoleEventHistory) (Utilities|Struct|MakeSGenoManagerEventList InstigatorIndex _returnvalue) 0)
  (Default|CallConsoleSequence InstigatorIndex _returnvalue))

(event Custom|BeginResetConsoleStates
  (CallFunction|ClearConsoleHistory)
  (Default|CallResetConsoleStates))

(fn ClearConsoleHistory
  (Utilities|Array|Clear (Variables|Default|GetConsoleEventHistory)))

(fn GetConsoleEventHistory
  (return (Variables|Default|GetConsoleEventHistory)))

(fn ConstructionScript)
