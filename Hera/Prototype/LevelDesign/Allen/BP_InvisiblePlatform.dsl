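; Invisible temple platform prototype: an invulnerable building actor that reveals itself while a player is in its trigger volume.
; BeginPlay creates dynamic materials for the floor, energy, corner and swirl meshes with initial fade values.
; On authority, overlap begin and end play or reverse one timeline driving those materials, plus a hint-mesh material never created in this graph.
; Remote instances only handle enter/exit audio; OnRep_IsInsideTrigger is empty.
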(event EventBeginPlay
  ; Init
  (Variables|BuildingActor|SetbIsInvulnerable true)
  (Variables|Actor|SetbCanBeDamaged false)
  ; Set Fade/Dynamic Material
  (bind _returnvalue (Rendering|Material|CreateDynamicMaterialInstance (Variables|Default|GetFloor%20Piece) 0 "/Hera/Prototype/LevelDesign/Allen/Materials/M_Fade.M_Fade" "None" :!class /Script/Engine.PrimitiveComponent))
  (bind _output_get (Variables|Default|SetDynamic%20Material _returnvalue))
  (Rendering|Material|SetScalarParameterValue _output_get "Fade Amount" 1.0 :!class /Script/Engine.MaterialInstanceDynamic)
  (bind _returnvalue_2 (Rendering|Material|CreateDynamicMaterialInstance (Variables|Default|GetSM_FX_TemplePlatform_Magic) 0 "/Hera/FXEnvironment/Temple/Materials/MI_Env_Temple_PlatformEnergy.MI_Env_Temple_PlatformEnergy" "None" :!class /Script/Engine.PrimitiveComponent))
  (bind _output_get_2 (Variables|Default|SetDynamic%20FX%20Platform%20Material _returnvalue_2))
  (Rendering|Material|SetScalarParameterValue _output_get_2 "GlobalOpacity" 0.0 :!class /Script/Engine.MaterialInstanceDynamic)
  (Rendering|SetVisibility (Variables|Default|GetSM_FX_Platform_Reveal_Swirl) true true)
  (bind _returnvalue_3 (Rendering|Material|CreateDynamicMaterialInstance (Variables|Default|GetSM_FX_Temple_PlatformCornerFX) 0 "/Hera/FXEnvironment/Temple/Materials/MI_Env_Temple_PlatformCornerMesh.MI_Env_Temple_PlatformCornerMesh" "None" :!class /Script/Engine.PrimitiveComponent))
  (bind _output_get_3 (Variables|Default|SetDynamic%20FX%20Corner%20Material _returnvalue_3))
  (Rendering|Material|SetScalarParameterValue _output_get_3 "GlobalOpacity" 0.0 :!class /Script/Engine.MaterialInstanceDynamic)
  (bind _returnvalue_4 (Rendering|Material|CreateDynamicMaterialInstance (Variables|Default|GetSM_FX_Platform_Reveal_Swirl) 0 "/Hera/FXEnvironment/Temple/Materials/MI_Env_Temple_PlatformEnergy.MI_Env_Temple_PlatformEnergy" "None" :!class /Script/Engine.PrimitiveComponent))
  (bind _output_get_4 (Variables|Default|SetDynamic%20FX%20Reveal%20Material _returnvalue_4))
  (Rendering|Material|SetScalarParameterValue _output_get_4 "GlobalOpacity" 0.0 :!class /Script/Engine.MaterialInstanceDynamic)
  (Rendering|Material|SetScalarParameterValue _output_get_4 "X_Tile" 0.25 :!class /Script/Engine.MaterialInstanceDynamic))

(event OnComponentBeginOverlap(TriggerVolume)
  (params (OverlappedComponent :type Primitive%20Component%20Object%20Reference) (OtherActor :type Actor%20Object%20Reference) (OtherComp :type Primitive%20Component%20Object%20Reference) (OtherBodyIndex :type Integer) (bFromSweep :type Boolean) (SweepResult :type Hit%20Result%20Structure%20%28by%20ref%29))
  ; Trigger Event (Reveal Platforms)
  (Networking|SwitchHasAuthority
    (:Authority
      (bind _asfort_player_pawn_athena (Utilities|Casting|CastToFortPlayerPawnAthena OtherActor))
      (Variables|Default|SetCurrent%20Pawn _asfort_player_pawn_athena)
      (bind _output_get (Variables|Default|SetIs%20Inside%20Trigger true))
      (if _output_get
        ; Reveal/Hide Timeline
        (bind (_direction _fade_track _fade_track_inverse _fade_track_revealmesh) (Timeline|Invisible%20Timeline Play)
          (:Update
            (Rendering|Material|SetScalarParameterValue (Variables|Default|GetDynamic%20Material) "Fade Amount" _fade_track :!class /Script/Engine.MaterialInstanceDynamic)
            (Rendering|Material|SetScalarParameterValue (Variables|Default|GetDynamic%20FX%20Platform%20Material) "GlobalOpacity" _fade_track_inverse :!class /Script/Engine.MaterialInstanceDynamic)
            (Rendering|Material|SetScalarParameterValue (Variables|Default|GetDynamic%20FX%20Corner%20Material) "GlobalOpacity" _fade_track_inverse :!class /Script/Engine.MaterialInstanceDynamic)
            (Rendering|Material|SetScalarParameterValue (Variables|Default|GetDynamic%20FX%20Reveal%20Material) "GlobalOpacity" _fade_track_revealmesh :!class /Script/Engine.MaterialInstanceDynamic)
            (Rendering|Material|SetScalarParameterValue (Variables|Default|GetDynamic%20FX%20HintMesh%20Material) "GlobalOpacity" _fade_track_inverse :!class /Script/Engine.MaterialInstanceDynamic)))
        (else
          (Timeline|Invisible%20Timeline Reverse))))
    (:Remote
      (Audio|Components|Audio|Play (Variables|Default|GetEnter%20Audio) 0.0))))

(event OnComponentEndOverlap(TriggerVolume)
  (params (OverlappedComponent :type Primitive%20Component%20Object%20Reference) (OtherActor :type Actor%20Object%20Reference) (OtherComp :type Primitive%20Component%20Object%20Reference) (OtherBodyIndex :type Integer))
  ; Exit Event (Hide Platforms)
  (Networking|SwitchHasAuthority
    (:Authority
      (Utilities|Casting|CastToFortPlayerPawnAthena OtherActor)
      (bind _output_get (Variables|Default|SetIs%20Inside%20Trigger false))
      (if _output_get
        ; Trigger Event (Reveal Platforms)
        ; Reveal/Hide Timeline
        (Timeline|Invisible%20Timeline Play)
        (else
          (Timeline|Invisible%20Timeline Reverse))))
    (:Remote
      (Audio|Components|Audio|Play (Variables|Default|GetExit%20Audio) 0.0)
      (Audio|Components|Audio|Stop (Variables|Default|GetEnter%20Audio)))))

(fn OnRep_IsInsideTrigger)

(fn ConstructionScript)
