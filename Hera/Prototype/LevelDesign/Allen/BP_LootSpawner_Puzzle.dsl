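; Puzzle loot spawner, an invulnerable building actor that pays out loot once when its trigger volume is first overlapped.
; On authority it casts the overlapper to a player pawn, deactivates the linked pressure plate elevator and spawns loot-tier pickups at the arrow.
; Each pickup is tossed along the arrow's forward vector with random spread.
; The remote branch plays a pirate chest sound at the pawn and a Niagara effect at the arrow.
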(event EventBeginPlay
  ; Init
  (Variables|BuildingActor|SetbIsInvulnerable true)
  (Variables|Actor|SetbCanBeDamaged false))

(event OnComponentBeginOverlap(TriggerVolume)
  (params (OverlappedComponent :type Primitive%20Component%20Object%20Reference) (OtherActor :type Actor%20Object%20Reference) (OtherComp :type Primitive%20Component%20Object%20Reference) (OtherBodyIndex :type Integer) (bFromSweep :type Boolean) (SweepResult :type Hit%20Result%20Structure%20%28by%20ref%29))
  (bind _arrow (Variables|Default|GetArrow))
  (bind _returnvalue (Transformation|GetForwardVector _arrow))
  (bind _toss_velocity (Variables|Default|GetToss%20Velocity))
  (bind _returnvalue_2 (Math|Conversions|ToVector(Float) _toss_velocity))
  (bind _returnvalue_3 (* _returnvalue _returnvalue_2))
  (Utilities|FlowControl|DoOnce false
    (:Completed
      (Networking|SwitchHasAuthority
        (:Authority
          (bind _asfort_player_pawn_athena (Utilities|Casting|CastToFortPlayerPawnAthena OtherActor))
          (Variables|Default|SetCurrentPawn _asfort_player_pawn_athena)
          (Class|BPPressurePlateElevatorShort|DeactivatePlatform (Variables|Default|GetPlatform%20Ref))
          (Variables|Default|SetImpulse _returnvalue_3)
          (bind _returnvalue_4 (Inventory|SpawnPickupInWorldWithLootTier (Variables|Default|GetLoot%20Tier%20to%20Fire) (Transformation|GetWorldLocation _arrow) 0 0 true true _returnvalue_3))
          (Variables|Default|SetLoot _returnvalue_4)
          (for _array_element (Variables|Default|GetLoot)
            (bind _asfort_pickup_athena (Utilities|Casting|CastToFortPickupAthena _array_element))
            (Pickup|TossPickupinDirection _asfort_pickup_athena (+ (Math|Vector|MakeVector (* (Variables|Default|GetToss%20Spread%20Multiplier) (Math|Random|RandomFloatinRange -1e+02 1e+02)) (* (Math|Random|RandomFloatinRange -1e+02 1e+02) (Variables|Default|GetToss%20Spread%20Multiplier)) (* (Math|Random|RandomFloatinRange 5e+01 1.5e+02) (Variables|Default|GetToss%20Spread%20Multiplier))) (* (Transformation|GetForwardVector (Variables|Default|GetArrow)) (Variables|Default|GetToss%20Velocity))) self false false)
            (Development|PrintString "SPAWN LOOT" true true "(R=0.000000,G=0.660000,B=1.000000,A=1.000000)" 5.0)))
        (:Remote
          (bind _asfort_player_pawn_athena_2 (Utilities|Casting|CastToFortPlayerPawnAthena OtherActor))
          (Audio|PlaySoundatLocation "/Game/Sounds/Foley_Loot/Containers/Pirate_Booty/Pirate_Booty_Beacon_Activate_Cue.Pirate_Booty_Beacon_Activate_Cue" (Transformation|GetActorLocation _asfort_player_pawn_athena_2))
          (Niagara|SpawnSystematLocation "/Game/Effects/Fort_Effects/Gameplay/PirateChests/NS_Loot_Chest.NS_Loot_Chest" (Transformation|GetWorldLocation (Variables|Default|GetArrow))))))))

(fn ConstructionScript)
