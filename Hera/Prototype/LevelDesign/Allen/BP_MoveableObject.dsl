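; Moveable Fortnite prototype prop that rides a referenced zipline spline and bursts out loot when its box is overlapped.
; On BeginPlay it becomes invulnerable, then a timeline scaled by Time to Complete drives its location and rotation along the spline.
; The one-shot overlap splits by authority: the server spawns tiered pickups tossed along the arrow with random spread, while remotes play chest sound and Niagara effect.
; The Loot Spawned timer event is empty.
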(event EventBeginPlay
  ; Set Invulnerable/Move Object on Start
  (Variables|BuildingActor|SetbIsInvulnerable true)
  (Variables|Actor|SetbCanBeDamaged false)
  (CallFunction|MoveObject))

(event Custom|Move%20Object
  ; Timeline
  (Components|Timeline|SetPlayRate (Variables|BP_MoveableObject|GetMover_Timeline) (/ 1.0 (Variables|Default|GetTime%20to%20Complete)))
  (bind (_direction _move_track) (Timeline|Mover_Timeline Play)
    (:Update
      ; Update Location/Rotation
      (Transformation|SetActorLocationAndRotation (Spline|GetLocationatDistanceAlongSpline (Variables|/Script/FortniteGame.FortAthenaSplineZipline|GetZiplineSpline (Variables|Default|GetSpline%20Ref)) (Math|Float|Lerp 0.0 (Spline|GetSplineLength (Variables|/Script/FortniteGame.FortAthenaSplineZipline|GetZiplineSpline (Variables|Default|GetSpline%20Ref))) _move_track) "World") (Spline|GetRotationatDistanceAlongSpline (Variables|/Script/FortniteGame.FortAthenaSplineZipline|GetZiplineSpline (Variables|Default|GetSpline%20Ref)) (Math|Float|Lerp 0.0 (Spline|GetSplineLength (Variables|/Script/FortniteGame.FortAthenaSplineZipline|GetZiplineSpline (Variables|Default|GetSpline%20Ref))) _move_track) "World")))))

(event OnComponentBeginOverlap(Box)
  (params (OverlappedComponent :type Primitive%20Component%20Object%20Reference) (OtherActor :type Actor%20Object%20Reference) (OtherComp :type Primitive%20Component%20Object%20Reference) (OtherBodyIndex :type Integer) (bFromSweep :type Boolean) (SweepResult :type Hit%20Result%20Structure%20%28by%20ref%29))
  (bind _toss_velocity (Variables|Default|GetToss%20Velocity))
  (bind _returnvalue (Math|Conversions|ToVector(Float) _toss_velocity))
  (bind _arrow (Variables|Default|GetArrow))
  (bind _returnvalue_2 (Transformation|GetForwardVector _arrow))
  (bind _returnvalue_3 (* _returnvalue_2 _returnvalue))
  ; Trigger Events (Loot)
  (Networking|SwitchHasAuthority
    (:Authority
      (bind _asfort_player_pawn_athena (Utilities|Casting|CastToFortPlayerPawnAthena OtherActor))
      (Variables|Default|SetCurrentPawn _asfort_player_pawn_athena)
      (Utilities|FlowControl|DoOnce false
        (:Completed
          (Variables|Default|SetImpulse _returnvalue_3)
          (bind _returnvalue_4 (Inventory|SpawnPickupInWorldWithLootTier (Variables|Default|GetLoot%20Tier%20To%20Fire) (Transformation|GetWorldLocation _arrow) 0 0 true true _returnvalue_3))
          (Variables|Default|SetLoot _returnvalue_4)
          (for _array_element (Variables|Default|GetLoot)
            (bind _asfort_pickup_athena (Utilities|Casting|CastToFortPickupAthena _array_element))
            (Pickup|TossPickupinDirection _asfort_pickup_athena (+ (Math|Vector|MakeVector (* (Variables|Default|GetToss%20Spread%20Multiplier) (Math|Random|RandomFloatinRange -4e+02 4e+02)) (* (Variables|Default|GetToss%20Spread%20Multiplier) (Math|Random|RandomFloatinRange -4e+02 4e+02)) (* (Variables|Default|GetToss%20Spread%20Multiplier) (Math|Random|RandomFloatinRange 1.5e+02 3.5e+02))) (* (Transformation|GetForwardVector (Variables|Default|GetArrow)) (Variables|Default|GetToss%20Velocity))) self false false)
            (Utilities|Time|SetTimerbyEvent (delegate Custom|Loot%20Spawned) 15.0)))))
    (:Remote
      (bind _asfort_player_pawn_athena_2 (Utilities|Casting|CastToFortPlayerPawnAthena OtherActor))
      (Utilities|FlowControl|DoOnce false
        (:Completed
          (Audio|PlaySoundatLocation "/Game/Sounds/Foley_Loot/Containers/Pirate_Booty/Pirate_Booty_Chest_Open_Cue.Pirate_Booty_Chest_Open_Cue" (Transformation|GetActorLocation _asfort_player_pawn_athena_2))
          (Niagara|SpawnSystematLocation "/Game/Effects/Fort_Effects/Gameplay/PirateChests/NS_Loot_Chest.NS_Loot_Chest" (Transformation|GetWorldLocation (Variables|Default|GetArrow)))
          (Utilities|Time|SetTimerbyEvent 0 15.0))))))

(event Custom|Loot%20Spawned)

(fn ConstructionScript
  (Components|StaticMesh|SetStaticMesh (Variables|Default|GetStaticMesh) (Variables|Default|GetObject%20Mesh)))
