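; Prototype damageable key item that roams a boat spline, hidden by a fade material and leaving timer-driven, ground-traced footstep decals and FX.
; LOS trigger overlap reverses the movement timeline.
; Box overlap stops it, with a polled player check driving the reveal fade and a GameplayEffect on the pawn; leaving resumes movement and hides it.
; Below quarter health it stops, spawns a pickup on authority, plays client-side FX and swaps to a broken-pot mesh.
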(event EventBeginPlay
  ; Init/ Spawn Item and Move
  (Variables|BuildingActor|SetbIsInvulnerable false)
  (Variables|Actor|SetbCanBeDamaged true)
  (Health|SetHealth self (Variables|Default|GetHealth%20))
  ; Set Invis Material
  (bind _returnvalue (Rendering|Material|CreateDynamicMaterialInstance (Variables|Default|GetStaticMesh) 0 "/Hera/Prototype/LevelDesign/Allen/Materials/M_Fade.M_Fade" "None" :!class /Script/Engine.PrimitiveComponent))
  (bind _output_get (Variables|Default|SetDynamic%20Material _returnvalue))
  (Rendering|Material|SetScalarParameterValue _output_get "Fade Amount" 1.0 :!class /Script/Engine.MaterialInstanceDynamic)
  ; Move Key Item on Spline
  (CallFunction|MoveObject)
  ; Footsteps Timer/Handle
  (bind _returnvalue_2 (Utilities|Time|SetTimerbyEvent (delegate Custom|Footstep%20Udate) (Variables|Default|GetFootsteps%20Update%20Time) true))
  (Variables|Default|SetFootsteps%20Timer%20Handle _returnvalue_2))

(event Custom|Footstep%20Udate
  (bind _self self)
  (bind _returnvalue (Transformation|GetVelocity _self))
  (bind _returnvalue_2 (Math|Vector|Normalize _returnvalue))
  (bind _returnvalue_3 (Math|Vector|VectorLength _returnvalue_2))
  ; Footsteps Decal/FX set up
  (if (> _returnvalue_3 0)
    (bind _output_get (Variables|Default|SetFootsteps%20Update%20Time%20Elapsed (+ (Variables|Default|GetFootsteps%20Update%20Time) (Variables|Default|GetFootsteps%20Update%20Time%20Elapsed))))
    ; Line Check/Spawn Effect
    (if (>= _output_get (/ _returnvalue_3 (Variables|Default|GetSteps%20Per%20Velocity)))
      (bind (_outhit _returnvalue_4) (Collision|LineTraceForObjects (Transformation|GetWorldLocation (Variables|Default|GetArrow)) (+ (Transformation|GetWorldLocation (Variables|Default|GetArrow)) (* "0.000000,0.000000,-1.000000" 100)) (Utilities|Array|MakeArray "ObjectTypeQuery1" "ObjectTypeQuery2")))
      (if _returnvalue_4
        (CallFunction|SpawnEffect (.ImpactPoint (Collision|BreakHitResult _outhit)))
        (Variables|Default|SetFootsteps%20Update%20Time%20Elapsed 0.0)))
    (else
      (Variables|Default|SetFootsteps%20Update%20Time%20Elapsed 0.0))))

(event Custom|Move%20Object
  ; Timeline/Spline Ref
  (Components|Timeline|SetPlayRate (Variables|BP_MovingKeyItem|GetMover_Timeline) (/ 1.0 (Variables|Default|GetTime%20To%20Complete)))
  (bind (_direction _move_track) (Timeline|Mover_Timeline Play)
    (:Update
      (Transformation|SetActorLocationAndRotation (Spline|GetLocationatDistanceAlongSpline (Variables|/Game/Athena/Artemis/Environments/Blueprints/Foundations/Artemis_BoatSpline.Artemis_BoatSpline_C|GetBoatSpline (Variables|Default|GetSpline%20Ref)) (Math|Float|Lerp 0.0 (Spline|GetSplineLength (Variables|/Game/Athena/Artemis/Environments/Blueprints/Foundations/Artemis_BoatSpline.Artemis_BoatSpline_C|GetBoatSpline (Variables|Default|GetSpline%20Ref))) _move_track) "World") (Spline|GetRotationatDistanceAlongSpline (Variables|/Game/Athena/Artemis/Environments/Blueprints/Foundations/Artemis_BoatSpline.Artemis_BoatSpline_C|GetBoatSpline (Variables|Default|GetSpline%20Ref)) (Math|Float|Lerp 0.0 (Spline|GetSplineLength (Variables|/Game/Athena/Artemis/Environments/Blueprints/Foundations/Artemis_BoatSpline.Artemis_BoatSpline_C|GetBoatSpline (Variables|Default|GetSpline%20Ref))) _move_track) "World")))))

(event OnComponentBeginOverlap(LOSTrigger)
  (params (OverlappedComponent :type Primitive%20Component%20Object%20Reference) (OtherActor :type Actor%20Object%20Reference) (OtherComp :type Primitive%20Component%20Object%20Reference) (OtherBodyIndex :type Integer) (bFromSweep :type Boolean) (SweepResult :type Hit%20Result%20Structure%20%28by%20ref%29))
  ; Reverse Trigger (Run away!)
  (Networking|SwitchHasAuthority
    (:Authority
      (Utilities|Casting|CastToFortPlayerPawnAthena OtherActor)
      (Transformation|SetActorRelativeRotation "0.000000,180.000000,0.000000")
      ; Timeline/Spline Ref
      (Timeline|Mover_Timeline Reverse))))

(event OnComponentBeginOverlap(BoxCollision)
  (params (OverlappedComponent :type Primitive%20Component%20Object%20Reference) (OtherActor :type Actor%20Object%20Reference) (OtherComp :type Primitive%20Component%20Object%20Reference) (OtherBodyIndex :type Integer) (bFromSweep :type Boolean) (SweepResult :type Hit%20Result%20Structure%20%28by%20ref%29))
  ; Trigger Events
  (Networking|SwitchHasAuthority
    (:Authority
      (bind _asfort_player_pawn_athena (Utilities|Casting|CastToFortPlayerPawnAthena OtherActor))
      (Variables|Default|SetCurrent%20Pawn _asfort_player_pawn_athena)
      (Utilities|FlowControl|Sequence
        (:then_0
          ; Timeline/Spline Ref
          (Timeline|Mover_Timeline Stop))
        (:then_1
          ; Timer/Player Check
          (bind _returnvalue (Utilities|Time|SetTimerbyEvent (delegate Custom|PlayerCheck) 1.0 true))
          (Variables|Default|SetTimer%20Handle _returnvalue)
          (bind _player_is_in_volume (CallFunction|IsPlayerinVolume (Collision|GetOverlappingActors (Variables|Default|GetBox%20Collision) "/Script/FortniteGame.FortPlayerPawnAthena" :!class /Script/Engine.PrimitiveComponent)))
          (Development|PrintString "Player is Inside Trigger")
          ; Apply GE/Check Timer
          (bind _output_get (Variables|Default|SetIsInsideTrigger _player_is_in_volume))
          (if _output_get
            (Environment|ApplyGlobalEnvironmentGameplayEffecttoActor (Variables|Default|GetCurrent%20Pawn) "/Game/Athena/Environments/Blueprints/DudeBro/GE_Athena_FallDmgImmune_FX.GE_Athena_FallDmgImmune_FX_C")
            ; Reveal/Hide Timeline
            (bind (_direction _fade_track) (Timeline|Invisible%20Timeline Play)
              (:Update
                (Rendering|Material|SetScalarParameterValue (Variables|Default|GetDynamic%20Material) "Fade Amount" _fade_track :!class /Script/Engine.MaterialInstanceDynamic)))
            (else
              (Timeline|Invisible%20Timeline Reverse))))))))

(event Custom|PlayerCheck
  ; Timer/Player Check
  (bind _player_is_in_volume (CallFunction|IsPlayerinVolume (Collision|GetOverlappingActors (Variables|Default|GetBox%20Collision) "/Script/FortniteGame.FortPlayerPawnAthena" :!class /Script/Engine.PrimitiveComponent)))
  (Development|PrintString "Player is Inside Trigger")
  ; Apply GE/Check Timer
  (bind _output_get (Variables|Default|SetIsInsideTrigger _player_is_in_volume))
  (if _output_get
    (Environment|ApplyGlobalEnvironmentGameplayEffecttoActor (Variables|Default|GetCurrent%20Pawn) "/Game/Athena/Environments/Blueprints/DudeBro/GE_Athena_FallDmgImmune_FX.GE_Athena_FallDmgImmune_FX_C")
    ; Reveal/Hide Timeline
    (Timeline|Invisible%20Timeline Play)
    (else
      (Timeline|Invisible%20Timeline Reverse))))

(event EventOnDamageServer
  (params (Damage :type Float%20%28single-precision%29) (DamageTags :type Gameplay%20Tag%20Container%20Structure%20%28by%20ref%29) (Momentum :type Vector :default "0, 0, 0") (HitInfo :type Hit%20Result%20Structure%20%28by%20ref%29) (InstigatedBy :type Controller%20Object%20Reference) (DamageCauser :type Actor%20Object%20Reference) (EffectContext :type Gameplay%20Effect%20Context%20Handle%20Structure))
  (bind (_bblockinghit _binitialoverlap _time _distance _location) (Collision|BreakHitResult HitInfo))
  ; Damage Logic
  (if (< (Health|GetHealthPercent self) 0.25)
    (Utilities|FlowControl|DoOnce false
      (:Completed
        (bind _output_get (Variables|Default|SetPot%20Is%20Damaged true))
        (CallFunction|PlayDamagedAudioandFX _location)
        (Utilities|Time|PauseTimerbyHandle (Variables|Default|GetTimer%20Handle))
        ; Apply GE/Check Timer
        (Utilities|FlowControl|Sequence
          (:then_0
            ; Timeline/Spline Ref
            (Timeline|Mover_Timeline Stop))
          (:then_1
            ; Spawn Loot
            (if _output_get
              (Utilities|FlowControl|DoOnce false
                (:Completed
                  (Networking|SwitchHasAuthority
                    (:Authority
                      (Audio|Components|Audio|Stop (Variables|Default|GetAudio))
                      (Inventory|SpawnPickupInWorldWithLevel (Variables|Default|GetItem%20Definition) 0 1 (Transformation|GetWorldLocation (Variables|Default|GetArrow)) "0.000000,0.000000,100.000000" 0 false)
                      (Development|PrintString "Spawn Loot" true true "(R=0.000000,G=0.660000,B=1.000000,A=1.000000)" 1e+01))
                    (:Remote
                      (Audio|PlaySoundatLocation "/Game/Sounds/Foley_Loot/Containers/Pirate_Booty/Pirate_Booty_Beacon_Activate_Cue.Pirate_Booty_Beacon_Activate_Cue" (Transformation|GetWorldLocation (Variables|Default|GetStaticMesh)))
                      (Niagara|SpawnSystematLocation "/Game/Effects/Fort_Effects/Gameplay/PirateChests/NS_Loot_Chest.NS_Loot_Chest" (Transformation|GetWorldLocation (Variables|Default|GetBillboard))))))))))))))

(event OnComponentEndOverlap(BoxCollision)
  (params (OverlappedComponent :type Primitive%20Component%20Object%20Reference) (OtherActor :type Actor%20Object%20Reference) (OtherComp :type Primitive%20Component%20Object%20Reference) (OtherBodyIndex :type Integer))
  ; Leave Trigger Events
  (Networking|SwitchHasAuthority
    (:Authority
      (Utilities|Casting|CastToFortPlayerPawnAthena OtherActor)
      (if (== (Utilities|Array|Length (Collision|GetOverlappingActors (Variables|Default|GetBox%20Collision) "/Script/FortniteGame.FortPlayerPawnAthena" :!class /Script/Engine.PrimitiveComponent)) 0)
        (bind _output_get (Variables|Default|SetIsInsideTrigger false))
        (if _output_get
          ; Apply GE/Check Timer
          ; Reveal/Hide Timeline
          (Timeline|Invisible%20Timeline Play)
          (else
            (Utilities|FlowControl|Sequence
              (:then_0
                (Utilities|Time|PauseTimerbyHandle (Variables|Default|GetTimer%20Handle))
                ; Timeline/Spline Ref
                (Timeline|Mover_Timeline Play))
              (:then_1
                (Timeline|Invisible%20Timeline Reverse)))))))))

(event Custom|PlayDamagedAudioAndFX
  (params (HitVector :type Vector))
  ; Client Cosmetics
  (Networking|SwitchHasAuthority
    (:Remote
      (Audio|PlaySoundatLocation "/Game/Sounds/Fort_GamePlay_Sounds/Player/Zap_Electricity_Revive_Cue.Zap_Electricity_Revive_Cue" HitVector)
      (Niagara|SpawnSystematLocation "/Game/Effects/Fort_Effects/Gameplay/Player/Tactical/NS_Player_HealthShield_Damage_Effect_TacElim.NS_Player_HealthShield_Damage_Effect_TacElim" HitVector))))

(fn IsPlayerInVolume
  (params (Players :type Array%20of%20Actor%20Object%20References))
  (bind _array_element (Utilities|Array|ForEachLoopwithBreak Players)
    (:LoopBody
      (bind _asfort_player_pawn_athena (Utilities|Casting|CastToFortPlayerPawnAthena _array_element))
      (Utilities|Casting|CastToFortPlayerControllerAthena (Pawn|GetController _asfort_player_pawn_athena))
      (LocalVariables|SetIs%20in%20Volume true))
    (:Completed
      _))
  (return (LocalVariables|GetIs%20in%20Volume)))

(fn OnRep_PotIsDamaged
  (if (Variables|Default|GetPot%20Is%20Damaged)
    (Components|StaticMesh|SetStaticMesh (Variables|Default|GetStaticMesh) "/Game/Environments/Sets/JungleTemple/Meshes/S_JungleTemple_Pots_Broken.S_JungleTemple_Pots_Broken")))

(fn SpawnEffect
  (params (Location :type Vector))
  (bind _returnvalue (Actor|GetOwner))
  (bind _returnvalue_2 (Transformation|GetActorRotation _returnvalue))
  (bind _returnvalue_3 (Math|Rotator|MakeRotator 0.0 0.0 9e+01))
  (bind _returnvalue_4 (Math|Rotator|CombineRotators _returnvalue_2 _returnvalue_3))
  (Game|SpawnActorfromClass (Variables|Default|GetFootstepDecalClass) (Math|Transform|MakeTransform Location _returnvalue_4) "Undefined" "MultiplyWithRoot" _returnvalue 0)
  (Effects|Components|ParticleSystem|SpawnFXAtLocation "/Game/Athena/Items/QuestParents/PropQuestActor/NS_PropQuest.NS_PropQuest" Location _returnvalue_4))

(fn ConstructionScript)
