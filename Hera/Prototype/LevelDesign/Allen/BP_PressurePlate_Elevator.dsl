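; Pressure plate elevator prototype, set invulnerable at start, that reacts to Fortnite player pawns overlapping its TriggerVolume.
; On authority, entry sets IsinTrigger and plays a switch-press timeline, then after a short delay an elevator timeline raises or lowers the Billboard component.
; Exit clears the flag and reverses both timelines, again delaying the elevator.
; Remote instances play activate or deactivate sounds and call a Niagara spawn at the Arrow.
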(event EventBeginPlay
  ; Init
  (Variables|BuildingActor|SetbIsInvulnerable true)
  (Variables|Actor|SetbCanBeDamaged false))

(event OnComponentBeginOverlap(TriggerVolume)
  (params (OverlappedComponent :type Primitive%20Component%20Object%20Reference) (OtherActor :type Actor%20Object%20Reference) (OtherComp :type Primitive%20Component%20Object%20Reference) (OtherBodyIndex :type Integer) (bFromSweep :type Boolean) (SweepResult :type Hit%20Result%20Structure%20%28by%20ref%29))
  ; Trigger Event
  (Networking|SwitchHasAuthority
    (:Authority
      (bind _asfort_player_pawn_athena (Utilities|Casting|CastToFortPlayerPawnAthena OtherActor))
      (Variables|Default|SetCurrentPawn _asfort_player_pawn_athena)
      (bind _output_get (Variables|Default|SetIsinTrigger true))
      (if _output_get
        (Utilities|FlowControl|Sequence
          (:then_0
            ; Switch/Button Timeline
            (bind (_direction _switchactivate) (Timeline|SwitchPress_Timeline Play)
              (:Update
                (Transformation|SetRelativeLocation (Variables|Default|GetSwitch) (Math|Vector|MakeVector 0.0 0.0 _switchactivate) false false))))
          (:then_1
            (Utilities|FlowControl|Delay 0.5)
            ; Elevator/Platform Timeline
            (bind (_direction_2 _elevatorposition) (Timeline|Activate_Elevator Play)
              (:Update
                (Transformation|SetRelativeLocation (Variables|Default|GetBillboard) (Math|Vector|MakeVector 0.0 0.0 _elevatorposition) false false)
                (Development|PrintString "Activate Elevator" true true "(R=0.000000,G=0.660000,B=1.000000,A=1.000000)" 1e+01)))))
        (else
          (Utilities|FlowControl|Sequence
            (:then_0
              (Timeline|SwitchPress_Timeline Reverse))
            (:then_1
              (Timeline|Activate_Elevator Reverse))))))
    (:Remote
      (Audio|Components|Audio|Play (Variables|Default|GetActivate%20Sound) 0.0)
      (Niagara|SpawnSystematLocation 0 (Transformation|GetWorldLocation (Variables|Default|GetArrow))))))

(event OnComponentEndOverlap(TriggerVolume)
  (params (OverlappedComponent :type Primitive%20Component%20Object%20Reference) (OtherActor :type Actor%20Object%20Reference) (OtherComp :type Primitive%20Component%20Object%20Reference) (OtherBodyIndex :type Integer))
  ; Exit Event
  (Networking|SwitchHasAuthority
    (:Authority
      (Utilities|Casting|CastToFortPlayerPawnAthena OtherActor)
      (bind _output_get (Variables|Default|SetIsinTrigger false))
      (if _output_get
        (Development|PrintString "Reverse Direction" true true "(R=0.000000,G=0.660000,B=1.000000,A=1.000000)" 1e+01)
        (Utilities|FlowControl|Sequence
          (:then_0
            ; Trigger Event
            ; Switch/Button Timeline
            (Timeline|SwitchPress_Timeline Play))
          (:then_1
            ; Elevator/Platform Timeline
            (Timeline|Activate_Elevator Play)))
        (else
          (Utilities|FlowControl|Sequence
            (:then_0
              (Timeline|SwitchPress_Timeline Reverse))
            (:then_1
              (Utilities|FlowControl|Delay 0.5)
              (Timeline|Activate_Elevator Reverse))))))
    (:Remote
      (Audio|Components|Audio|Play (Variables|Default|GetDeactivate%20Sound) 0.0)
      (Niagara|SpawnSystematLocation 0 (Transformation|GetWorldLocation (Variables|Default|GetArrow))))))

(fn IsPlayerInVolume
  (params (Players :type Array%20of%20Actor%20Object%20References))
  (bind _array_element (Utilities|Array|ForEachLoopwithBreak Players)
    (:LoopBody
      (bind _asfort_player_pawn_athena (Utilities|Casting|CastToFortPlayerPawnAthena _array_element))
      (Utilities|Casting|CastToFortPlayerControllerAthena (Pawn|GetController _asfort_player_pawn_athena))
      (LocalVariables|SetIs%20In%20Volume true))
    (:Completed
      (return (LocalVariables|GetIs%20In%20Volume)))))

(fn OnRep_IsinTrigger)

(fn ConstructionScript)
