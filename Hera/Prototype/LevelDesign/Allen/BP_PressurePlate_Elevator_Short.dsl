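; Pressure plate elevator prototype that presses a switch and moves a platform while a Fortnite player pawn stands in the trigger.
; Overlap authority branches drive switch and delayed elevator timelines, lerping the Billboard from its start to an offset end and reversing on exit.
; Remote branches play sounds and Niagara bursts, while a helper toggles dust and smoke.
; A one-shot Deactivate Platform event, not called here, locks the end position and disables the trigger.
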(event EventBeginPlay
  (bind _billboard (Variables|Default|GetBillboard))
  (bind _relativelocation (Variables|/Script/Engine.SceneComponent|GetRelativeLocation _billboard))
  ; Init
  (Variables|BuildingActor|SetbIsInvulnerable true)
  (Variables|Actor|SetbCanBeDamaged false)
  (Variables|Default|SetDoor%20Start%20Loc _relativelocation)
  (Variables|Default|SetDoor%20End%20Loc (+ _relativelocation (Variables|Default|GetEnd%20Position%20Variable))))

(event OnComponentBeginOverlap(SwitchVolume)
  (params (OverlappedComponent :type Primitive%20Component%20Object%20Reference) (OtherActor :type Actor%20Object%20Reference) (OtherComp :type Primitive%20Component%20Object%20Reference) (OtherBodyIndex :type Integer) (bFromSweep :type Boolean) (SweepResult :type Hit%20Result%20Structure%20%28by%20ref%29))
  ; Trigger Event
  (Networking|SwitchHasAuthority
    (:Authority
      (bind _asfort_player_pawn_athena (Utilities|Casting|CastToFortPlayerPawnAthena OtherActor))
      (Variables|Default|SetCurrentPawn _asfort_player_pawn_athena)
      (bind _output_get (Variables|Default|SetIsinTrigger true))
      (CallFunction|Enable/DisableFX)
      (if _output_get
        (Utilities|FlowControl|Sequence
          (:then_0
            ; Switch/Button Timeline
            (bind (_direction _switchactivate) (Timeline|SwitchPress_Timeline Play)
              (:Update
                (Transformation|SetRelativeLocation (Variables|Default|GetSwitch) (Math|Vector|MakeVector 0.0 0.0 _switchactivate) false false))))
          (:then_1
            (Utilities|FlowControl|Delay 0.5)
            ; Elevator/Platform Timeline
            (bind (_direction_2 _lerp) (Timeline|Activate_Elevator Play :NewTime (Variables|Default|GetTravel%20Time))
              (:Update
                (Transformation|SetRelativeLocation (Variables|Default|GetBillboard) (Math|Vector|Lerp(Vector) (Variables|Default|GetDoor%20Start%20Loc) (Variables|Default|GetDoor%20End%20Loc) _lerp) false false))
              (:Finished
                (if (Variables|Default|GetIsComplete)
                  (Development|PrintString "Stop Timeline" true true "(R=0.000000,G=0.660000,B=1.000000,A=1.000000)" 1e+01)
                  ; Exit Event
                  (Timeline|Activate_Elevator Stop))))))
        (else
          (Utilities|FlowControl|Sequence
            (:then_0
              (Timeline|SwitchPress_Timeline Reverse))
            (:then_1
              (Timeline|Activate_Elevator Reverse))))))
    (:Remote
      (Audio|Components|Audio|Play (Variables|Default|GetActivate%20Sound) 0.0)
      (Audio|Components|Audio|Play (Variables|Default|GetAudio%20Looping) 0.0)
      (Niagara|SpawnSystematLocation 0 (Transformation|GetWorldLocation (Variables|Default|GetArrow))))))

(event OnComponentEndOverlap(SwitchVolume)
  (params (OverlappedComponent :type Primitive%20Component%20Object%20Reference) (OtherActor :type Actor%20Object%20Reference) (OtherComp :type Primitive%20Component%20Object%20Reference) (OtherBodyIndex :type Integer))
  ; Exit Event
  (Networking|SwitchHasAuthority
    (:Authority
      (Utilities|Casting|CastToFortPlayerPawnAthena OtherActor)
      (bind _output_get (Variables|Default|SetIsinTrigger false))
      (CallFunction|Enable/DisableFX)
      (if _output_get
        (Utilities|FlowControl|Sequence
          (:then_0
            ; Trigger Event
            ; Switch/Button Timeline
            (Timeline|SwitchPress_Timeline Play))
          (:then_1
            ; Elevator/Platform Timeline
            (Timeline|Activate_Elevator Play)))
        (else
          (Utilities|FlowControl|Sequence
            (:then_0
              (Timeline|SwitchPress_Timeline Reverse))
            (:then_1
              (Utilities|FlowControl|Delay 0.5)
              (Timeline|Activate_Elevator Reverse))))))
    (:Remote
      (Audio|Components|Audio|Play (Variables|Default|GetDeactivate%20Sound) 0.0)
      (Niagara|SpawnSystematLocation 0 (Transformation|GetWorldLocation (Variables|Default|GetArrow)))
      (Audio|Components|Audio|Stop (Variables|Default|GetAudio%20Looping)))))

(event Custom|Deactivate%20Platform
  ; Platform has reached the End (Disable Platform, Spawn Loot)
  (Utilities|FlowControl|DoOnce false
    (:Completed
      (bind _output_get (Variables|Default|SetIsComplete true))
      (if _output_get
        (Components|Activation|Activate (Variables|Default|GetFinished%20FX) false)
        ; Disable/Clean Up
        (Components|Activation|Deactivate (Variables|Default|GetFX))
        (Transformation|SetRelativeLocation (Variables|Default|GetBillboard) (Variables|Default|GetDoor%20End%20Loc) false false)
        (Development|PrintString "Platform Deactivated")
        (Collision|SetCollisionEnabled (Variables|Default|GetSwitch%20Volume) "PhysicsOnly")))))

(fn Enable/DisableFX
  (bind _dust (Variables|Default|GetDust))
  (if (Variables|Default|GetIsinTrigger)
    (Components|Activation|Activate _dust false)
    (Components|Activation|Activate (Variables|Default|GetFloor%20Smoke) false)
    (else
      (Components|Activation|Deactivate _dust)
      (Components|Activation|Deactivate (Variables|Default|GetFloor%20Smoke1)))))

(fn IsPlayerInVolume
  (params (Players :type Array%20of%20Actor%20Object%20References))
  (bind _array_element (Utilities|Array|ForEachLoopwithBreak Players)
    (:LoopBody
      (bind _asfort_player_pawn_athena (Utilities|Casting|CastToFortPlayerPawnAthena _array_element))
      (Utilities|Casting|CastToFortPlayerControllerAthena (Pawn|GetController _asfort_player_pawn_athena))
      (LocalVariables|SetIs%20In%20Volume true))
    (:Completed
      _))
  (return (LocalVariables|GetIs%20In%20Volume)))

(fn OnRep_IsComplete)

(fn OnRep_IsinTrigger)

(fn ConstructionScript)
