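; Pressure switch prototype, an invulnerable building actor that opens linked secret doors while a Fortnite player pawn stands on it.
; On authority, OverlapBox begin and end events set IsInTrigger and call OpenVault or SwitchDeactivated on DoorstoActivate.
; The SwitchActivate timeline plays or reverses to lerp the switch between its start and offset end positions.
; Remote instances play activate or deactivate audio and spawn Niagara effects at the arrow.
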(event EventBeginPlay
  (bind _switch (Variables|Default|GetSwitch))
  (bind _relativelocation (Variables|/Script/Engine.SceneComponent|GetRelativeLocation _switch))
  ; Init
  (Variables|BuildingActor|SetbIsInvulnerable true)
  (Variables|Actor|SetbCanBeDamaged false)
  (Variables|Default|SetSwitchStartLocation _relativelocation)
  (Variables|Default|SetSwitchEndLocation (+ _relativelocation (Variables|Default|GetLocation%20Variable))))

(event OnComponentBeginOverlap(OverlapBox)
  (params (OverlappedComponent :type Primitive%20Component%20Object%20Reference) (OtherActor :type Actor%20Object%20Reference) (OtherComp :type Primitive%20Component%20Object%20Reference) (OtherBodyIndex :type Integer) (bFromSweep :type Boolean) (SweepResult :type Hit%20Result%20Structure%20%28by%20ref%29))
  ; Overlap Events
  (Networking|SwitchHasAuthority
    (:Authority
      (bind _asfort_player_pawn_athena (Utilities|Casting|CastToFortPlayerPawnAthena OtherActor))
      (Variables|Default|SetCurrentPawn _asfort_player_pawn_athena)
      (bind _output_get (Variables|Default|SetIsInTrigger true))
      (Class|BPSecretDoorPressurePlate|OpenVault (Variables|Default|GetDoorstoActivate) 0)
      (if _output_get
        ; Timeline Event
        (bind (_direction _switchposition) (Timeline|SwitchActivate Play)
          (:Update
            (Transformation|SetRelativeLocation (Variables|Default|GetSwitch) (Math|Vector|Lerp(Vector) (Variables|Default|GetSwitchStartLocation) (Variables|Default|GetSwitchEndLocation) _switchposition) false false)))
        (else
          (Timeline|SwitchActivate Reverse))))
    (:Remote
      (Audio|Components|Audio|Play (Variables|Default|GetActivate%20Sound) 0.0)
      (Niagara|SpawnSystematLocation 0 (Transformation|GetWorldLocation (Variables|Default|GetArrow))))))

(event OnComponentEndOverlap(OverlapBox)
  (params (OverlappedComponent :type Primitive%20Component%20Object%20Reference) (OtherActor :type Actor%20Object%20Reference) (OtherComp :type Primitive%20Component%20Object%20Reference) (OtherBodyIndex :type Integer))
  ; Overlap Events
  (Networking|SwitchHasAuthority
    (:Authority
      (Utilities|Casting|CastToFortPlayerPawnAthena OtherActor)
      (bind _output_get (Variables|Default|SetIsInTrigger false))
      (Class|BPSecretDoorPressurePlate|SwitchDeactivated (Variables|Default|GetDoorstoActivate))
      (if _output_get
        ; Timeline Event
        (Timeline|SwitchActivate Play)
        (else
          (Timeline|SwitchActivate Reverse))))
    (:Remote
      (Audio|Components|Audio|Play (Variables|Default|GetDeactivate%20Sound) 0.0)
      (Niagara|SpawnSystematLocation 0 (Transformation|GetWorldLocation (Variables|Default|GetArrow))))))

(fn IsPlayerInVolume
  (params (Players :type Array%20of%20Actor%20Object%20References))
  (bind _array_element (Utilities|Array|ForEachLoopwithBreak Players)
    (:LoopBody
      (bind _asfort_player_pawn_athena (Utilities|Casting|CastToFortPlayerPawnAthena _array_element))
      (Utilities|Casting|CastToFortPlayerControllerAthena (Pawn|GetController _asfort_player_pawn_athena))
      (LocalVariables|SetIs%20In%20Volume true))
    (:Completed
      _))
  (return (LocalVariables|GetIs%20In%20Volume)))

(fn OnRep_IsInTrigger)

(fn ConstructionScript)
