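; Pressure plate switch actor that fires linked temple statue traps when a Fortnite player pawn overlaps its trigger box.
; On the authority, overlap waits a configurable delay, activates the traps, then plays a timeline that lerps the switch slightly downward.
; Ending the overlap deactivates the traps and reverses the timeline, while remote clients only play sound and Niagara effects.
; BeginPlay makes the actor invulnerable and caches the switch's start and end positions.
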(event EventBeginPlay
  (bind _switch (Variables|Default|GetSwitch))
  (bind _relativelocation (Variables|/Script/Engine.SceneComponent|GetRelativeLocation _switch))
  ; Init
  (Variables|BuildingActor|SetbIsInvulnerable true)
  (Variables|Actor|SetbCanBeDamaged false)
  (Variables|Default|SetSwitchStartLocation _relativelocation)
  (Variables|Default|SetSwitchEndLocation (+ _relativelocation "0.000000,0.000000,-8.000000")))

(event OnComponentBeginOverlap(OverlapBox)
  (params (OverlappedComponent :type Primitive%20Component%20Object%20Reference) (OtherActor :type Actor%20Object%20Reference) (OtherComp :type Primitive%20Component%20Object%20Reference) (OtherBodyIndex :type Integer) (bFromSweep :type Boolean) (SweepResult :type Hit%20Result%20Structure%20%28by%20ref%29))
  ; Overlap Events
  (Networking|SwitchHasAuthority
    (:Authority
      (bind _asfort_player_pawn_athena (Utilities|Casting|CastToFortPlayerPawnAthena OtherActor))
      (Variables|Default|SetCurrentPawn _asfort_player_pawn_athena)
      (Variables|Default|SetIsInTrigger true)
      (Utilities|FlowControl|Delay (Variables|Default|GetActivate%20Trap%20Delay))
      (Class|BPTempleStatueTrap|ActivateStatueTrap(Fire) (Variables|Default|GetTraps%20To%20Activate))
      ; Timeline Event
      (bind (_direction _switchposition) (Timeline|SwitchActivate Play)
        (:Update
          (Transformation|SetRelativeLocation (Variables|Default|GetSwitch) (Math|Vector|Lerp(Vector) (Variables|Default|GetSwitchStartLocation) (Variables|Default|GetSwitchEndLocation) _switchposition) false false))))
    (:Remote
      (Audio|Components|Audio|Play (Variables|Default|GetActivate%20Sound) 0.0)
      (Niagara|SpawnSystematLocation 0 (Transformation|GetWorldLocation (Variables|Default|GetArrow))))))

(event OnComponentEndOverlap(OverlapBox)
  (params (OverlappedComponent :type Primitive%20Component%20Object%20Reference) (OtherActor :type Actor%20Object%20Reference) (OtherComp :type Primitive%20Component%20Object%20Reference) (OtherBodyIndex :type Integer))
  ; Overlap Events
  (Networking|SwitchHasAuthority
    (:Authority
      (Utilities|Casting|CastToFortPlayerPawnAthena OtherActor)
      (Variables|Default|SetIsInTrigger false)
      (Class|BPTempleStatueTrap|DeactivateStatueTrap (Variables|Default|GetTraps%20To%20Activate))
      ; Timeline Event
      (Timeline|SwitchActivate Reverse))
    (:Remote
      (Audio|Components|Audio|Play (Variables|Default|GetDeactivate%20Sound) 0.0)
      (Niagara|SpawnSystematLocation 0 (Transformation|GetWorldLocation (Variables|Default|GetArrow))))))

(fn ConstructionScript)
