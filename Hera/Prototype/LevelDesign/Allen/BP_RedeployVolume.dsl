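; Prototype redeploy volume, an invulnerable building actor that flings Fortnite player pawns high into the air when they overlap its trigger.
; On authority, it stores the overlapping pawn, launches it steeply upward with a forward push, and sets it to ignore the next fall damage.
; It then applies the directional-launch redeploy gameplay effect to that pawn.
; Remote instances only play a wind gust sound at the pawn's location.
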(event EventBeginPlay
  ; Make Actor Invulnerable
  (Variables|BuildingActor|SetbIsInvulnerable true)
  (Variables|Actor|SetbCanBeDamaged false))

(event OnComponentBeginOverlap(TriggerVolume)
  (params (OverlappedComponent :type Primitive%20Component%20Object%20Reference) (OtherActor :type Actor%20Object%20Reference) (OtherComp :type Primitive%20Component%20Object%20Reference) (OtherBodyIndex :type Integer) (bFromSweep :type Boolean) (SweepResult :type Hit%20Result%20Structure%20%28by%20ref%29))
  (bind _current_pawn (Variables|Default|GetCurrent%20Pawn))
  ; Trigger Events
  (Networking|SwitchHasAuthority
    (:Authority
      (bind _asfort_player_pawn_athena (Utilities|Casting|CastToFortPlayerPawnAthena OtherActor))
      (bind _output_get (Variables|Default|SetCurrent%20Pawn _asfort_player_pawn_athena))
      (Character|LaunchCharacter _output_get (Math|Vector|MakeVector (.x (* (Transformation|GetActorForwardVector _output_get) (Math|Conversions|ToVector(Float) 1e+03))) (.y (* (Transformation|GetActorForwardVector _output_get) (Math|Conversions|ToVector(Float) 1e+03))) 8e+03) true true)
      (Variables|/Script/FortniteGame.FortPawn|SetbIgnoreNextFallingDamage true _current_pawn)
      (Environment|ApplyGlobalEnvironmentGameplayEffecttoActor _current_pawn "/Game/Athena/Environments/Blueprints/EnvironmentalLaunch/Generic/GE_DirectionalLaunched_Applied_Generic_Redeploy.GE_DirectionalLaunched_Applied_Generic_Redeploy_C"))
    (:Remote
      (bind _asfort_player_pawn_athena_2 (Utilities|Casting|CastToFortPlayerPawnAthena OtherActor))
      (Audio|PlaySoundatLocation "/AmbientAudioContent/SoundCues/OneShots/Spot_Desert_Wind_Whispy_Gust_Cue.Spot_Desert_Wind_Whispy_Gust_Cue" (Transformation|GetActorLocation _asfort_player_pawn_athena_2)))))

(fn ConstructionScript)
