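; Prototype sauna building actor that a player switches on by spending a key item, turning it into a campfire-style healing area.
; Interaction requires the key, sets ButtonIsPressed and enables HealVolume collision; its OnRep shows lights and glowing rocks, plus FX and audio.
; On the server, overlap starts a one-second looping timer that reapplies the campfire heal effect to the last entering pawn, pausing when the volume empties.
; A debug flag spawns a key pickup.
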(event EventBeginPlay
  ; Init / Debug Key
  (Variables|BuildingActor|SetbIsInvulnerable true)
  (Fortnite|SetCanbeDamaged self)
  (Networking|SwitchHasAuthority
    (:Authority
      (if (Variables|Default|GetDebugSpawnKey)
        (Utilities|FlowControl|Delay 1.0)
        (Inventory|SpawnPickupInWorldWithLevel (Variables|Default|GetItem%20Definition) 0 1 (Transformation|GetWorldLocation (Variables|Default|GetArrow)))
        ; disable Heal Volume until Interact
        (Collision|SetCollisionEnabled (Variables|Default|GetHeal%20Volume) "NoCollision")
        (Development|PrintString "Heal Volume is OFF" true true "(R=0.000000,G=0.660000,B=1.000000,A=1.000000)" 5.0)))))

(event OnComponentBeginOverlap(HealVolume)
  (params (OverlappedComponent :type Primitive%20Component%20Object%20Reference) (OtherActor :type Actor%20Object%20Reference) (OtherComp :type Primitive%20Component%20Object%20Reference) (OtherBodyIndex :type Integer) (bFromSweep :type Boolean) (SweepResult :type Hit%20Result%20Structure%20%28by%20ref%29))
  ; Trigger Events
  (Networking|SwitchHasAuthority
    (:Authority
      (bind _asfort_player_pawn_athena (Utilities|Casting|CastToFortPlayerPawnAthena OtherActor))
      (bind _output_get (Variables|Default|SetCurrentPawn _asfort_player_pawn_athena))
      (bind _returnvalue (Utilities|Time|SetTimerbyEvent (delegate Custom|Player%20Check) 1.0 true))
      (Variables|Default|SetTimer%20Handle _returnvalue)
      (bind _player_is_in_volume (CallFunction|IsPlayerinVolume (Collision|GetOverlappingActors (Variables|Default|GetHeal%20Volume) "/Script/FortniteGame.FortPlayerPawnAthena" :!class /Script/Engine.PrimitiveComponent)))
      (Development|PrintString "Player is Inside Trigger" true true "(R=0.000000,G=0.660000,B=1.000000,A=1.000000)" 5.0)
      (bind _output_get_2 (Variables|Default|SetIsInsideTrigger _player_is_in_volume))
      (if _output_get_2
        (Environment|ApplyGlobalEnvironmentGameplayEffecttoActor _output_get "/Game/Athena/Items/Traps/Campfire/Blueprint/GE_Athena_Campfire_Heal.GE_Athena_Campfire_Heal_C")
        (Development|PrintString "HEALING" true true "(R=0.000000,G=0.660000,B=1.000000,A=1.000000)" 5.0)))))

(event Custom|Player%20Check
  ; Trigger Events
  (bind _player_is_in_volume (CallFunction|IsPlayerinVolume (Collision|GetOverlappingActors (Variables|Default|GetHeal%20Volume) "/Script/FortniteGame.FortPlayerPawnAthena" :!class /Script/Engine.PrimitiveComponent)))
  (Development|PrintString "Player is Inside Trigger" true true "(R=0.000000,G=0.660000,B=1.000000,A=1.000000)" 5.0)
  (bind _output_get (Variables|Default|SetIsInsideTrigger _player_is_in_volume))
  (if _output_get
    (Environment|ApplyGlobalEnvironmentGameplayEffecttoActor (stale OnComponentBeginOverlap(HealVolume) _output_get) "/Game/Athena/Items/Traps/Campfire/Blueprint/GE_Athena_Campfire_Heal.GE_Athena_Campfire_Heal_C")
    (Development|PrintString "HEALING" true true "(R=0.000000,G=0.660000,B=1.000000,A=1.000000)" 5.0)))

(event Interaction|EventBlueprintOnInteract
  (params (InteractingPawn :type Fort%20Pawn%20Object%20Reference) (InteractionBeingAttempted :type EInteractionBeingAttempted%20Enum :default "FirstInteraction"))
  ; Button Interaction Logic
  (Variables|Default|SetButtonCanBePressed false)
  (CallFunction|RemoveKeyItem InteractingPawn)
  (CallFunction|ButtonPressedEvent)
  (CallFunction|HealVolumeIsOn)
  (Development|PrintString "Button Pressed - Event on Server" true true "(R=0.000000,G=0.660000,B=1.000000,A=1.000000)" 5.0))

(event OnComponentEndOverlap(HealVolume)
  (params (OverlappedComponent :type Primitive%20Component%20Object%20Reference) (OtherActor :type Actor%20Object%20Reference) (OtherComp :type Primitive%20Component%20Object%20Reference) (OtherBodyIndex :type Integer))
  ; Trigger Events
  (Networking|SwitchHasAuthority
    (:Authority
      (Utilities|Casting|CastToFortPlayerPawnAthena OtherActor)
      (if (== (Utilities|Array|Length (Collision|GetOverlappingActors (Variables|Default|GetHeal%20Volume) "/Script/FortniteGame.FortPlayerPawnAthena" :!class /Script/Engine.PrimitiveComponent)) 0)
        (Variables|Default|SetIsInsideTrigger false)
        (Utilities|Time|PauseTimerbyHandle (Variables|Default|GetTimer%20Handle))))))

(event Custom|ButtonPressedEvent
  ; Button Interaction Event
  (Networking|SwitchHasAuthority
    (:Authority
      (Variables|Default|SetButtonIsPressed true))
    (:Remote
      (Audio|Components|Audio|Play (Variables|Default|GetAudio%20Interact) 0.0)
      (Niagara|SpawnSystematLocation 0 (Transformation|GetWorldLocation (Variables|Default|GetArrow))))) ; authority = server, Remote = Client
)

(fn BlueprintCanInteract
  (params (InteractingPawn :type Fort%20Pawn%20Object%20Reference) (InteractionBeingAttempted :type EInteractionBeingAttempted%20Enum :default "FirstInteraction") (InteractionType :type TInteractionType%20Enum :default "IT_NoInteraction"))
  (bind _asfort_player_pawn_athena (Utilities|Casting|CastToFortPlayerPawnAthena InteractingPawn))
  (bind _asfort_player_controller_athena (Utilities|Casting|CastToFortPlayerControllerAthena (Pawn|GetController _asfort_player_pawn_athena)))
  (return (and (Variables|Default|GetButtonCanBePressed) (>= (Inventory|GetItemQuantityonInventoryOwner (Fortnite|GetFortPlayerControllerfromActor _asfort_player_controller_athena) (Variables|Default|GetKeyItemDefinition)) 1))))

(fn BlueprintGetInteractionString
  (params (InteractingPawn :type Fort%20Pawn%20Object%20Reference) (InteractionBeingAttempted :type EInteractionBeingAttempted%20Enum :default "FirstInteraction"))
  (return (Variables|Default|GetButtonInteractionText)))

(fn HealVolumeIsOn
  (if (Variables|Default|GetButtonIsPressed)
    (Collision|SetCollisionEnabled (Variables|Default|GetHeal%20Volume) "QueryOnly")
    (Development|PrintString "Heal Volume ON" true true "(R=0.000000,G=0.660000,B=1.000000,A=1.000000)" 5.0)))

(fn IsPlayerInVolume
  (params (Players :type Array%20of%20Actor%20Object%20References))
  (bind _array_element (Utilities|Array|ForEachLoopwithBreak Players :!label _label)
    (:LoopBody
      (bind _asfort_player_pawn_athena (Utilities|Casting|CastToFortPlayerPawnAthena _array_element))
      (Utilities|Casting|CastToFortPlayerControllerAthena (Pawn|GetController _asfort_player_pawn_athena))
      (LocalVariables|SetIsInVolume true)
      (goto _label Break))
    (:Completed
      (return (LocalVariables|GetIsInVolume)))))

(fn OnRep_ButtonIsPressed
  (if (Variables|Default|GetButtonIsPressed)
    ; Enable Point Light
    (Rendering|SetVisibility (Variables|Default|GetPointLight) true false)
    ; Enable Occupied/Vacant Lights
    (Rendering|SetVisibility (Variables|Default|GetOccupied/Vacant%20Light%2001) true false)
    (Rendering|SetVisibility (Variables|Default|GetOccupied/Vacant%20Light%2002) true false)
    ; Spawn Rock Mesh
    (Rendering|SetVisibility (Variables|Default|GetStaticMesh_Rocks) true false)
    ; Swap Material
    (Rendering|Material|SetMaterial (Variables|Default|GetStaticMesh_Rocks) 0 "/Game/Environments/FrontEnd/Materials/BP12/Emissive_Red_Midas_FE.Emissive_Red_Midas_FE")
    ; Enable FX
    (Components|Activation|Activate (Variables|Default|GetFX) false)
    ; Play Looping Audio
    (Audio|Components|Audio|Play (Variables|Default|GetAudio%20Loop) 0.0)
    (Niagara|SpawnSystematLocation 0 (Transformation|GetWorldLocation (Variables|Default|GetArrow)))
    (else
      (Rendering|SetVisibility (Variables|Default|GetPointLight) false false))))

(fn OnRep_IsInsideTrigger)

(fn RemoveKeyItem
  (params (Pawn :type Fort%20Pawn%20Object%20Reference))
  (bind _asfort_player_controller_athena (Utilities|Casting|CastToFortPlayerControllerAthena (Pawn|GetController Pawn)))
  (Inventory|RemoveItemfromInventoryOwner _asfort_player_controller_athena (Variables|Default|GetItem%20Definition)))

(fn ConstructionScript)
