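; One-way secret door prototype: an invulnerable building actor whose mesh slides open while a player pawn occupies its trigger volume.
; BeginPlay caches start and offset end positions, which the OpenDoor timeline lerps between.
; On authority, overlap begin plays the timeline with a moving-door Niagara effect, and overlap end reverses it.
; Remote instances play open or close sounds and spawn Niagara at an arrow, without checking the actor type; OnRep_IsInTrigger is empty.
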(event EventBeginPlay
  (bind _door_mesh (Variables|Default|GetDoor%20Mesh))
  (bind _relativelocation (Variables|/Script/Engine.SceneComponent|GetRelativeLocation _door_mesh))
  ; Init
  (Variables|BuildingActor|SetbIsInvulnerable true)
  (Variables|Actor|SetbCanBeDamaged false)
  (Variables|Default|SetDoorStartLocation _relativelocation)
  (Variables|Default|SetDoorEndLocation (+ _relativelocation (Variables|Default|GetDoor%20Loction%20Variable))))

(event OnComponentBeginOverlap(TriggerVolume)
  (params (OverlappedComponent :type Primitive%20Component%20Object%20Reference) (OtherActor :type Actor%20Object%20Reference) (OtherComp :type Primitive%20Component%20Object%20Reference) (OtherBodyIndex :type Integer) (bFromSweep :type Boolean) (SweepResult :type Hit%20Result%20Structure%20%28by%20ref%29))
  ; Trigger Door
  (Networking|SwitchHasAuthority
    (:Authority
      (Utilities|Casting|CastToFortPlayerPawnAthena OtherActor)
      (bind _output_get (Variables|Default|SetIsInTrigger true))
      (if _output_get
        (Components|Activation|Activate (Variables|Default|GetNS_Env_Temple_MovingDoor) false)
        (bind (_direction _doorposition) (Timeline|OpenDoor Play :NewTime (Variables|Default|GetTravel%20Time))
          (:Update
            (Transformation|SetRelativeLocation (Variables|Default|GetDoor%20Mesh) (Math|Vector|Lerp(Vector) (Variables|Default|GetDoorStartLocation) (Variables|Default|GetDoorEndLocation) _doorposition) false false))
          (:Finished
            (Components|Activation|Deactivate (Variables|Default|GetNS_Env_Temple_MovingDoor))))
        (else
          (Timeline|OpenDoor Reverse))))
    (:Remote
      (Audio|Components|Audio|Play (Variables|Default|GetDoor%20Open%20Sound) 0.0)
      (Niagara|SpawnSystematLocation 0 (Transformation|GetWorldLocation (Variables|Default|GetArrow))))))

(event OnComponentEndOverlap(TriggerVolume)
  (params (OverlappedComponent :type Primitive%20Component%20Object%20Reference) (OtherActor :type Actor%20Object%20Reference) (OtherComp :type Primitive%20Component%20Object%20Reference) (OtherBodyIndex :type Integer))
  ; Trigger Door
  (Networking|SwitchHasAuthority
    (:Authority
      (Utilities|Casting|CastToFortPlayerPawnAthena OtherActor)
      (bind _output_get (Variables|Default|SetIsInTrigger false))
      (if _output_get
        (Timeline|OpenDoor Play)
        (else
          (Timeline|OpenDoor Reverse))))
    (:Remote
      (Audio|Components|Audio|Play (Variables|Default|GetDoor%20Close%20Sound) 0.0)
      (Niagara|SpawnSystematLocation 0 (Transformation|GetWorldLocation (Variables|Default|GetArrow))))))

(fn OnRep_IsInTrigger)

(fn ConstructionScript)
