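; Secret vault door in a temple maze that slides open when enough pressure switches are active.
; OpenVault and SwitchDeactivated adjust a switch count and compare it to the required number, playing the OpenDoor timeline or reversing it after a delay.
; The timeline lerps the mesh by a set offset, with Niagara effects, camera shake and audio.
; When the timeline finishes, it broadcasts a custom verb message with the pawn for an accolade.
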(event EventBeginPlay
  (bind _door_mesh (Variables|Default|GetDoor%20Mesh))
  (bind _relativelocation (Variables|/Script/Engine.SceneComponent|GetRelativeLocation _door_mesh))
  ; Init
  (Variables|BuildingActor|SetbIsInvulnerable true)
  (Variables|Actor|SetbCanBeDamaged false)
  (Variables|Default|SetDoorStartLocation _relativelocation)
  (Variables|Default|SetDoorEndLocation (+ _relativelocation (Variables|Default|GetDoor%20Loction%20Variable))))

(event Custom|OpenVault
  (params (Pawn%20Ref :type Fort%20Pawn%20Object%20Reference))
  ; Open Vault Door
  (Math|Integer|IncrementInt (Variables|Default|GetPressureSwitchesActive)
    (:%20%20
      (if (>= (Variables|Default|GetPressureSwitchesActive) (Variables|Default|GetPressureSwitchesNeeded))
        (Variables|Default|SetPawn%20Ref Pawn%20Ref)
        (Components|Activation|Activate (Variables|Default|GetNS_Env_Temple_Maze_VaultOpen) false)
        (Components|Activation|Activate (Variables|Default|GetNS_Env_Temple_Maze_VaultOpen_Debris) false)
        (CameraShake|Start (Variables|Default|GetCameraShakeSource))
        ; Audio
        (Audio|Components|Audio|Play (Variables|Default|GetDoorOpenSound) 0.0)
        (bind (_direction _doorposition) (Timeline|OpenDoor Play :NewTime (Variables|Default|GetTravel%20Time))
          (:Update
            (Transformation|SetRelativeLocation (Variables|Default|GetDoor%20Mesh) (Math|Vector|Lerp(Vector) (Variables|Default|GetDoorStartLocation) (Variables|Default|GetDoorEndLocation) _doorposition) false false))
          (:Finished
            (Components|Activation|Deactivate (Variables|Default|GetNS_Env_Temple_Maze_VaultOpen_Debris))
            ;[0.15 0.15 0.15 0.5 15] Custom Verb Message to trigger the Accolade
            (CustomVerb|BroadcastCustomVerbMessage (CustomVerb|MakeCustomVerbMessage (Variables|Default|GetPawn%20Ref) "(TagName=\"CustomVerb.S42.Maze.VaultOpen\")"))))
        (else
          (Utilities|FlowControl|Delay (Variables|Default|GetDelay))
          (Timeline|OpenDoor Reverse))))))

(event Custom|SwitchDeactivated
  ; Open Vault Door
  (Math|Integer|DecrementInt (Variables|Default|GetPressureSwitchesActive)
    (:%20%20
      (if (>= (Variables|Default|GetPressureSwitchesActive) (Variables|Default|GetPressureSwitchesNeeded))
        (Timeline|OpenDoor Play)
        (else
          (Utilities|FlowControl|Delay (Variables|Default|GetDelay))
          ; Audio
          (Audio|Components|Audio|Stop (Variables|Default|GetDoorOpenSound))
          (Audio|Components|Audio|Play (Variables|Default|GetDoorCloseSound) 0.0)
          (Timeline|OpenDoor Reverse))))))

(fn ConstructionScript)
