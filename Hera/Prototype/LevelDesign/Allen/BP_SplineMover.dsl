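; Building actor that moves an array of actors along its spline at runtime, spaced by a fixed offset.
; BeginPlay makes it invulnerable and starts a timeline whose play rate is the inverse of Length.
; Each update converts timeline alpha to spline distance, adds index times offset, wraps overflow by one spline length, and sets each actor's location and rotation.
; The construction script pre-places the actors at their offsets.
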(event EventBeginPlay
  ; Make Actor Invulnerable
  (Variables|BuildingActor|SetbIsInvulnerable true)
  (Variables|Actor|SetbCanBeDamaged false)
  ; Move Objects/Play Timeline
  (Components|Timeline|SetPlayRate (Variables|BP_SplineMover|GetTimeline) (/ 1.0 (Variables|Default|GetLength)))
  (CallFunction|PlayTimeline))

(event Custom|Play%20Timeline
  (bind _spline (Variables|Default|GetSpline))
  (bind _returnvalue (Spline|GetSplineLength _spline))
  ; Spline Logic
  (bind (_direction _alpha) (Timeline|Timeline Play)
    (:Update
      (for (_array_element _array_index) (Variables|Default|GetActors)
        (bind _output_get (Variables|Default|SetDistance%20Along%20Spline (+ (Math|Float|Lerp 0.0 (Spline|GetSplineLength (Variables|Default|GetSpline)) _alpha) (* _array_index (Variables|Default|GetOffset%20Between%20Actors)))))
        (if (> _output_get _returnvalue)
          (Variables|Default|SetDistance%20Along%20Spline (- (Variables|Default|GetDistance%20Along%20Spline) _returnvalue)))
        (Transformation|SetActorLocationAndRotation _array_element (Spline|GetLocationatDistanceAlongSpline (Variables|Default|GetSpline) (Variables|Default|GetDistance%20Along%20Spline) "World") (Spline|GetRotationatDistanceAlongSpline (Variables|Default|GetSpline) (Variables|Default|GetDistance%20Along%20Spline) "World") false false)))))

(fn ConstructionScript
  (for (_array_element _array_index) (Variables|Default|GetActors)
    (Transformation|SetActorLocationAndRotation _array_element (Spline|GetLocationatDistanceAlongSpline (Variables|Default|GetSpline) (* (Variables|Default|GetOffset%20Between%20Actors) _array_index) "World") (Spline|GetRotationatDistanceAlongSpline (Variables|Default|GetSpline) (* (Variables|Default|GetOffset%20Between%20Actors) _array_index) "World") false false)))
