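; Temple tower brazier building actor, invulnerable, whose fire meshes, darken mesh, light and decal start hidden until activated.
; The activate event sets Is Active on authority, while remotes directly play audio and call a Niagara spawn at the Arrow.
; OnRep_IsActive toggles visuals and audio, repeating the spawn call when active.
; The construction script randomizes smoke mesh scale, UV offset and fade, and tints the decal cyan through dynamic materials.
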(event EventBeginPlay
  ; Init
  (Variables|BuildingActor|SetbIsInvulnerable true)
  (Variables|Actor|SetbCanBeDamaged false)
  ; Hide FX/Lights
  (Rendering|SetVisibility (Variables|Default|GetDecal) false false)
  (Rendering|SetVisibility (Variables|Default|GetFire_Mesh_Outer) false false)
  (Rendering|SetVisibility (Variables|Default|GetFire_Mesh_Inner) false false)
  (Rendering|SetVisibility (Variables|Default|GetDarken_FX_Mesh) false false)
  (Rendering|SetVisibility (Variables|Default|GetPointLight) false false)
  (Rendering|SetVisibility (Variables|Default|GetFire_Mesh_Flipbook) false false)
  (Development|PrintString "Deactivate Temple fire" true true "(R=0.000000,G=0.660000,B=1.000000,A=1.000000)" 1e+01))

(event Custom|Activate%20Temple%20Tower%20Brazier
  ; Custom Event - Activate (Enable FX/Lights/Audio)
  (Networking|SwitchHasAuthority
    (:Authority
      (Variables|Default|SetIs%20Active true)
      (Development|PrintString "Activate Temple Fire" true true "(R=0.000000,G=0.660000,B=1.000000,A=1.000000)" 1e+01))
    (:Remote
      (Audio|Components|Audio|Play (Variables|Default|GetAudio) 0.0)
      (Niagara|SpawnSystematLocation 0 (Transformation|GetWorldLocation (Variables|Default|GetArrow))))))

(fn OnRep_ButtonIsPressed)

(fn OnRep_IsActive
  (bind _fire_mesh_outer (Variables|Default|GetFire_Mesh_Outer))
  (bind _pointlight (Variables|Default|GetPointLight))
  (bind _decal (Variables|Default|GetDecal))
  (bind _fire_mesh_inner (Variables|Default|GetFire_Mesh_Inner))
  (bind _fire_mesh_flipbook (Variables|Default|GetFire_Mesh_Flipbook))
  (bind _audio (Variables|Default|GetAudio))
  (bind _darken_fx_mesh (Variables|Default|GetDarken_FX_Mesh))
  (if (Variables|Default|GetIs%20Active)
    (Rendering|SetVisibility _pointlight true false)
    (Rendering|SetVisibility _fire_mesh_inner true false)
    (Rendering|SetVisibility _fire_mesh_outer true false)
    (Rendering|SetVisibility _fire_mesh_flipbook true false)
    (Rendering|SetVisibility _darken_fx_mesh true false)
    (Rendering|SetVisibility _decal true false)
    (Audio|Components|Audio|Play _audio 0.0)
    (Niagara|SpawnSystematLocation 0 (Transformation|GetWorldLocation (Variables|Default|GetArrow)))
    (else
      (Rendering|SetVisibility _pointlight false false)
      (Rendering|SetVisibility _fire_mesh_inner false false)
      (Rendering|SetVisibility _fire_mesh_outer false false)
      (Rendering|SetVisibility _fire_mesh_flipbook false false)
      (Rendering|SetVisibility _darken_fx_mesh false false)
      (Rendering|SetVisibility _decal false false)
      (Audio|Components|Audio|Stop _audio))))

(fn ConstructionScript
  (bind _returnvalue (Math|Random|RandomFloat))
  (|CallParentFunction)
  (bind _returnvalue_2 (Rendering|Material|CreateDynamicMaterialInstance (Variables|Default|GetSmoke_Mesh) 0 (Rendering|Material|GetMaterial (Variables|Default|GetSmoke_Mesh) 0) "None" :!class /Script/Engine.PrimitiveComponent))
  (Rendering|Material|SetScalarParameterValue _returnvalue_2 "OffsetUVs" _returnvalue :!class /Script/Engine.MaterialInstanceDynamic)
  (Transformation|SetRelativeScale3D (Variables|Default|GetSmoke_Mesh) (* (Variables|/Script/Engine.SceneComponent|GetRelativeScale3D (Variables|Default|GetSmoke_Mesh)) (Math|Vector|Lerp(Vector) "0.700000,0.700000,0.700000" "1.300000,1.300000,1.000000" _returnvalue)))
  (Rendering|Material|SetScalarParameterValue _returnvalue_2 "MeshFade" (Math|Float|Lerp 0.0 0.6 _returnvalue) :!class /Script/Engine.MaterialInstanceDynamic)
  (bind _returnvalue_3 (Rendering|Components|Decal|CreateDynamicMaterialInstance (Variables|Default|GetDecal)))
  (Rendering|Material|SetVectorParameterValue _returnvalue_3 "EmissiveEdge" "(R=0.000000,G=0.000000,B=0.000000,A=1.000000)" :!class /Script/Engine.MaterialInstanceDynamic)
  (Rendering|Material|SetVectorParameterValue _returnvalue_3 "EmissiveCenter" "(R=0.000000,G=1.500000,B=1.500000,A=1.000000)" :!class /Script/Engine.MaterialInstanceDynamic))
