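; Invulnerable temple statue fire trap, switched by activate and deactivate events that toggle flames, light, audio and the damage volume.
; A timeline drives windup glow and decal emissive, reversing on deactivation before a Trap Duration delay shuts the trap off.
; On the server, player overlap applies damage and a fire affliction, repeated by a half-second timer on the last-entered pawn.
; The timer pauses once no player pawns remain inside.
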(event EventBeginPlay
  ; Init
  (Variables|BuildingActor|SetbIsInvulnerable true)
  (Variables|Actor|SetbCanBeDamaged false)
  ; Disable Damage Volume
  (Collision|SetCollisionEnabled (Variables|Default|GetDamage%20Volume) "NoCollision"))

(event Custom|Activate%20Statue%20Trap%20%28Fire%29
  ; Activate Trap
  (Variables|Default|SetTrapIsOn true)
  (CallFunction|Enable/DisableFX/Audio)
  (CallFunction|Enable/DisableDamageVolume)
  ; VFX
  (bind (_direction _windupglow _emissivemaskpower) (Timeline|VFX_Timeline Play)
    (:Update
      (Niagara|SetNiagaraVariable(Float) (Variables|Default|GetWindup%20Glow%20FX) "WindupGlow" _windupglow)
      (bind _returnvalue (Rendering|Components|Decal|CreateDynamicMaterialInstance (Variables|Default|GetDecal)))
      (Rendering|Material|SetScalarParameterValue _returnvalue "EmissiveMaskPower" _emissivemaskpower :!class /Script/Engine.MaterialInstanceDynamic))))

(event Custom|Deactivate%20Statue%20Trap
  ; Deactivate Trap
  (Utilities|FlowControl|Sequence
    (:then_0
      ; VFX
      (Timeline|VFX_Timeline Reverse))
    (:then_1
      (Utilities|FlowControl|Delay (Variables|Default|GetTrap%20Duration))
      (Variables|Default|SetTrapIsOn false)
      (CallFunction|Enable/DisableFX/Audio)
      (CallFunction|Enable/DisableDamageVolume))))

(event OnComponentBeginOverlap(DamageVolume)
  (params (OverlappedComponent :type Primitive%20Component%20Object%20Reference) (OtherActor :type Actor%20Object%20Reference) (OtherComp :type Primitive%20Component%20Object%20Reference) (OtherBodyIndex :type Integer) (bFromSweep :type Boolean) (SweepResult :type Hit%20Result%20Structure%20%28by%20ref%29))
  ; Collision Events
  (Networking|SwitchHasAuthority
    (:Authority
      (bind _asfort_player_pawn_athena (Utilities|Casting|CastToFortPlayerPawnAthena OtherActor)) ; this checks, on the server, if a player touched this, not a bot or wildlife
      (bind _output_get (Variables|Default|SetCurrentPawn _asfort_player_pawn_athena))
      (bind _returnvalue (Utilities|Time|SetTimerbyEvent (delegate Custom|Player%20Check) 0.5 true false 1.0))
      (Variables|Default|SetTimer%20Handle _returnvalue)
      (bind _player_is_in_volume (CallFunction|IsPlayerinVolume (Collision|GetOverlappingActors (Variables|Default|GetDamage%20Volume) "/Script/FortniteGame.FortPlayerPawnAthena" :!class /Script/Engine.PrimitiveComponent)))
      (Development|PrintString "Player Is In Damage Trigger")
      (bind _output_get_2 (Variables|Default|SetIsInsideTrigger _player_is_in_volume))
      (if _output_get_2
        (Environment|ApplyGlobalEnvironmentDamagetoActor _output_get (Variables|Default|GetDamage%20Float))
        (Environment|ApplyGlobalEnvironmentGameplayEffecttoActor _output_get "/Game/Items/Alteration_v2/GameplayAlterations/Affliction/GE_Alteration_Affliction_Fire.GE_Alteration_Affliction_Fire_C")))))

(event Custom|Player%20Check
  ; Collision Events
  (bind _player_is_in_volume (CallFunction|IsPlayerinVolume (Collision|GetOverlappingActors (Variables|Default|GetDamage%20Volume) "/Script/FortniteGame.FortPlayerPawnAthena" :!class /Script/Engine.PrimitiveComponent)))
  (Development|PrintString "Player Is In Damage Trigger")
  (bind _output_get (Variables|Default|SetIsInsideTrigger _player_is_in_volume))
  (if _output_get
    (Environment|ApplyGlobalEnvironmentDamagetoActor (stale OnComponentBeginOverlap(DamageVolume) _output_get) (Variables|Default|GetDamage%20Float))
    (Environment|ApplyGlobalEnvironmentGameplayEffecttoActor (stale OnComponentBeginOverlap(DamageVolume) _output_get) "/Game/Items/Alteration_v2/GameplayAlterations/Affliction/GE_Alteration_Affliction_Fire.GE_Alteration_Affliction_Fire_C")))

(event OnComponentEndOverlap(DamageVolume)
  (params (OverlappedComponent :type Primitive%20Component%20Object%20Reference) (OtherActor :type Actor%20Object%20Reference) (OtherComp :type Primitive%20Component%20Object%20Reference) (OtherBodyIndex :type Integer))
  ; Collision Events
  (Networking|SwitchHasAuthority
    (:Authority
      (Utilities|Casting|CastToFortPlayerPawnAthena OtherActor) ; this checks, on the server, if a player touched this, not a bot or wildlife
      (if (== (Utilities|Array|Length (Collision|GetOverlappingActors (Variables|Default|GetDamage%20Volume) "/Script/FortniteGame.FortPlayerPawnAthena" :!class /Script/Engine.PrimitiveComponent)) 0)
        (Variables|Default|SetIsInsideTrigger false)
        (Utilities|Time|PauseTimerbyHandle (Variables|Default|GetTimer%20Handle))
        (Development|PrintString "Pause Timer")))))

(fn Enable/DisableDamageVolume
  (bind _damage_volume (Variables|Default|GetDamage%20Volume))
  (if (Variables|Default|GetTrapIsOn)
    (Collision|SetCollisionEnabled _damage_volume "QueryOnly")
    (Development|PrintString "Damage Volume is On")
    (else
      (Collision|SetCollisionEnabled _damage_volume "NoCollision")
      (Development|PrintString "Damage Volume is OFF"))))

(fn Enable/DisableFX/Audio
  (bind _pointlight (Variables|Default|GetPointLight))
  (bind _flame_fx (Variables|Default|GetFlame%20FX))
  (bind _windup_glow_fx (Variables|Default|GetWindup%20Glow%20FX))
  (bind _loop_audio (Variables|Default|GetLoop%20Audio))
  (if (Variables|Default|GetTrapIsOn)
    ; Enable Fire Light
    (Rendering|SetVisibility _pointlight true false)
    ; Enable FX
    (Components|Activation|Activate _flame_fx true)
    (Components|Activation|Activate _windup_glow_fx true)
    ; Play Audio/ Start and Loop
    (Utilities|FlowControl|Sequence
      (:then_0
        (Audio|Components|Audio|Play (Variables|Default|GetStart%20Audio) 0.0))
      (:then_1
        (Audio|Components|Audio|Play _loop_audio 0.0)))
    (else
      (Rendering|SetVisibility _pointlight false false)
      (Components|Activation|Deactivate _flame_fx)
      (Components|Activation|Deactivate _windup_glow_fx)
      ; End Audio
      (Audio|Components|Audio|Stop _loop_audio)
      (Audio|Components|Audio|Play (Variables|Default|GetEnd%20Audio) 0.0))))

(fn IsPlayerInVolume
  (params (Players :type Array%20of%20Actor%20Object%20References))
  (bind _array_element (Utilities|Array|ForEachLoopwithBreak Players :!label _label)
    (:LoopBody
      (bind _asfort_player_pawn_athena (Utilities|Casting|CastToFortPlayerPawnAthena _array_element))
      (Utilities|Casting|CastToFortPlayerControllerAthena (Pawn|GetController _asfort_player_pawn_athena))
      (LocalVariables|SetIsInVolume true)
      (goto _label Break))
    (:Completed
      (return (LocalVariables|GetIsInVolume)))))

(fn OnRep_IsInsideTrigger)

(fn OnRep_TrapIsOn)

(fn ConstructionScript)
