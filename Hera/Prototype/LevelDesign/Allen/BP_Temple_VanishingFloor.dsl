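; Temple vault puzzle actor with a vanishing floor (Door) that opens once enough switches are active.
; Server-side proximity overlaps drive a Fade Timeline on the floor's dynamic material to reveal or hide it.
; The final switch fades the floor, shuts off the fountain, then after delays moves the water plane, broadcasts an accolade message and cleans up.
; RepNotify flags update lights, door visibility and collision, and fountain state.
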(event EventBeginPlay
  ; Init
  (Variables|BuildingActor|SetbIsInvulnerable true)
  (Variables|Actor|SetbCanBeDamaged false)
  ; Set Fade/Dynamic Material
  (bind _returnvalue (Rendering|Material|CreateDynamicMaterialInstance (Variables|Default|GetDoor) 0 "/Hera/Prototype/LevelDesign/Allen/Materials/M_Fade.M_Fade" "None" :!class /Script/Engine.PrimitiveComponent))
  (bind _output_get (Variables|Default|SetDynamic%20Material _returnvalue))
  (Rendering|Material|SetScalarParameterValue _output_get "Fade Amount" 1.0 :!class /Script/Engine.MaterialInstanceDynamic))

(event Custom|Activate%20Temple%20Vault
  (params (Pawn%20Ref :type Fort%20Pawn%20Object%20Reference))
  ; Open Temple Vault Event
  (Math|Integer|IncrementInt (Variables|Default|GetSwitches%20Active)
    (:%20%20
      (if (>= (Variables|Default|GetSwitches%20Active) (Variables|Default|GetSwitches%20Required))
        (bind _output_get (Variables|Default|SetPawn%20Ref Pawn%20Ref))
        (GameplayCue|ExecuteGameplayCue _output_get (Variables|Default|GetGameplay%20Cue%20Tag) (Utilities|Struct|MakeGameplayEffectContextHandle))
        ;; Fade Floor first then activate vault
        (Utilities|FlowControl|Sequence
          (:then_0
            (CallFunction|TurnOffFountain))
          (:then_1
            (Components|Timeline|ReversefromEnd (Variables|BP_Temple_VanishingFloor|GetFade%20Timeline)))
          (:then_2
            (Utilities|FlowControl|Delay (Variables|Default|GetVault%20Open%20Delay))
            (Variables|Default|SetTemple%20Vault%20is%20Active true)
            ; Remove Water
            (Class|BPTempleWaterPlaneEvent|MoveWaterPlane (Variables|Default|GetWater%20Plane%20Ref))
            (Development|PrintString "Vault is Open!" true true "(R=0.000000,G=0.660000,B=1.000000,A=1.000000)" 1e+01)
            ;[0.15 0.15 0.15 0.5 15] Custom Verb Message to trigger the Accolade
            (CustomVerb|BroadcastCustomVerbMessage (CustomVerb|MakeCustomVerbMessage (Variables|Default|GetPawn%20Ref) "(TagName=\"CustomVerb.S42.Temple.VaultOpen\")"))
            ; Clean up FX/Lighting
            (Utilities|FlowControl|Delay (Variables|Default|GetClean%20Up%20Delay))
            (CallFunction|CleanUpTempleEvent)
            (Development|PrintString "Clean Up Temple" true true "(R=0.000000,G=0.660000,B=1.000000,A=1.000000)" 1e+01)))))))

(event Custom|TurnOffFountain
  ; Vault Opening FX | Turn Off Fountain FX
  (Components|Activation|Activate (Variables|Default|GetMagic%20FX) false)
  (Niagara|SetNiagaraVariable(Bool) (Variables|Default|GetFX_Fountain) "KillFountain" true)
  (bind _returnvalue (Rendering|Material|CreateDynamicMaterialInstance (Variables|Default|GetSM_FountainWaterfalls) 0 "/Hera/FXEnvironment/Temple/Materials/MI_Env_Temple_Fountain_Water_Spline.MI_Env_Temple_Fountain_Water_Spline" "None" :!class /Script/Engine.PrimitiveComponent))
  (Utilities|FlowControl|Sequence
    (:then_0
      (Audio|AddEvent (GameInstanceSubsystems|GetMusicEventSubsystem) "/Hera/Audio/POI/Temple/MSS_MX_S42_POI_Temple_PuzzleSolve_OneShotSingle.MSS_MX_S42_POI_Temple_PuzzleSolve_OneShotSingle" "(TagName=\"MusicEventSubsystem.Event.POI.Gameplay\")" "(GameplayTags=)"))
    (:then_1
      (bind (_direction _waterfall_opacity _pointlight_intensity) (Timeline|Waterfall_Timeline Play)
        (:Update
          (Rendering|Material|SetScalarParameterValue _returnvalue "FinalOpacity" _waterfall_opacity :!class /Script/Engine.MaterialInstanceDynamic)
          (Rendering|Components|Light|SetIntensity (Variables|Default|GetPointLight1) _pointlight_intensity))
        (:Finished
          (Networking|SwitchHasAuthority
            (:Authority
              (Variables|Default|SetFountainIsEmpty true))))))))

(event Custom|Clean%20Up%20Temple%20Event
  ; Deactivate Temple Event
  (Networking|SwitchHasAuthority
    (:Authority
      (Variables|Default|SetTemple%20Vault%20is%20Active false)
      (Components|Activation|Deactivate (Variables|Default|GetFX_Fountain))
      (Rendering|SetVisibility (Variables|Default|GetSM_FountainWaterfalls) false false)
      (Components|Activation|Deactivate (Variables|Default|GetMagic%20FX)))))

(event OnComponentBeginOverlap(ProximityVisibility)
  (params (OverlappedComponent :type Primitive%20Component%20Object%20Reference) (OtherActor :type Actor%20Object%20Reference) (OtherComp :type Primitive%20Component%20Object%20Reference) (OtherBodyIndex :type Integer) (bFromSweep :type Boolean) (SweepResult :type Hit%20Result%20Structure%20%28by%20ref%29))
  ; Reveal Floor
  (Networking|SwitchHasAuthority
    (:Authority
      (bind _asfort_player_pawn_athena (Utilities|Casting|CastToFortPlayerPawnAthena OtherActor))
      (Variables|Default|SetCurrentPawn _asfort_player_pawn_athena)
      (bind _output_get (Variables|Default|SetIs%20Inside%20Trigger true))
      (if _output_get
        ; Fading Timeline
        (bind (_direction _fade_track) (Timeline|Fade%20Timeline Play)
          (:Update
            (Rendering|Material|SetScalarParameterValue (Variables|Default|GetDynamic%20Material) "Fade Amount" _fade_track :!class /Script/Engine.MaterialInstanceDynamic)))
        (else
          (Timeline|Fade%20Timeline Reverse))))))

(event OnComponentEndOverlap(ProximityVisibility)
  (params (OverlappedComponent :type Primitive%20Component%20Object%20Reference) (OtherActor :type Actor%20Object%20Reference) (OtherComp :type Primitive%20Component%20Object%20Reference) (OtherBodyIndex :type Integer))
  ; Hide Floor
  (Networking|SwitchHasAuthority
    (:Authority
      (Utilities|Casting|CastToFortPlayerPawnAthena OtherActor)
      (bind _output_get (Variables|Default|SetIs%20Inside%20Trigger false))
      (if _output_get
        ; Fading Timeline
        (Timeline|Fade%20Timeline Play)
        (else
          (Timeline|Fade%20Timeline Reverse))))))

(fn OnRep_FountainIsEmpty
  (if (Variables|Default|GetFountainIsEmpty)
    (Components|Activation|Deactivate (Variables|Default|GetFX_Fountain))
    (bind _returnvalue (Rendering|Material|CreateDynamicMaterialInstance (Variables|Default|GetSM_FountainWaterfalls) 0 "/Hera/FXEnvironment/Temple/Materials/MI_Env_Temple_Fountain_Water_Spline.MI_Env_Temple_Fountain_Water_Spline" "None" :!class /Script/Engine.PrimitiveComponent))
    (Rendering|Material|SetScalarParameterValue _returnvalue "FinalOpacity" 0.0 :!class /Script/Engine.MaterialInstanceDynamic)
    (Rendering|Components|Light|SetIntensity (Variables|Default|GetPointLight1) 0.0)))

(fn OnRep_IsInsideTrigger)

(fn OnRep_TempleVaultisActive
  (bind _pointlight1 (Variables|Default|GetPointLight1))
  (bind _pointlight2 (Variables|Default|GetPointLight2))
  (bind _pointlight3 (Variables|Default|GetPointLight3))
  (bind _door (Variables|Default|GetDoor))
  (bind _pointlight (Variables|Default|GetPointLight))
  (if (Variables|Default|GetTemple%20Vault%20is%20Active)
    ; Activate/Deactive Point Lights
    (Rendering|SetVisibility _pointlight true false)
    (Rendering|SetVisibility _pointlight1 true false)
    (Rendering|SetVisibility _pointlight2 true false)
    (Rendering|SetVisibility _pointlight3 true false)
    (Development|PrintString "Turn on FX")
    (Rendering|SetVisibility _door false false)
    (Collision|SetCollisionEnabled _door "NoCollision")
    (Development|PrintString "Remove Door/Collision")
    (else
      (Rendering|SetVisibility _pointlight false false)
      (Rendering|SetVisibility _pointlight1 false false)
      (Rendering|SetVisibility _pointlight2 false false)
      (Rendering|SetVisibility _pointlight3 false false))))

(fn ConstructionScript)
