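; Temple level-design prototype actor that moves a water plane between two relative positions when its Move Water Plane event is called.
; BeginPlay makes the actor invulnerable, stores the plane's start location and adds an offset variable for the end, using door-named variables.
; MoverTimeline lerps the plane between those points.
; Travel Time feeds only the timeline's New Time input, while execution enters through Play.
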(event EventBeginPlay
  (bind _plane (Variables|Default|GetPlane))
  (bind _relativelocation (Variables|/Script/Engine.SceneComponent|GetRelativeLocation _plane))
  ; Init
  (Variables|BuildingActor|SetbIsInvulnerable true)
  (Variables|Actor|SetbCanBeDamaged false)
  (Variables|Default|SetDoor%20Start%20Location _relativelocation)
  (Variables|Default|SetDoor%20End%20Location (+ _relativelocation (Variables|Default|GetDoor%20Location%20Variable))))

(event Custom|Move%20Water%20Plane
  ; Timeline/Custom Event
  (bind (_direction _lerp) (Timeline|MoverTimeline Play :NewTime (Variables|Default|GetTravel%20Time))
    (:Update
      (Transformation|SetRelativeLocation (Variables|Default|GetPlane) (Math|Vector|Lerp(Vector) (Variables|Default|GetDoor%20Start%20Location) (Variables|Default|GetDoor%20End%20Location) _lerp) false false))))

(fn ConstructionScript)
