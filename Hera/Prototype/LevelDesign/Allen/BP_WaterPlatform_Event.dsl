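; Prototype switch-driven water platform event that shifts a water cube and exclusion volume by set offsets to drain water, and can raise it again.
; BeginPlay caches endpoints; activation counts switches and, at the required number, authority runs a timeline lerping both and destroys the river body.
; Deactivation reverses the timeline after a delay.
; Authority handles movement and waterfall FX, while remotes play sounds and call Niagara spawns with no system assigned.
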(event EventBeginPlay
  (bind _waterexclusionref (Variables|Default|GetWaterExclusionRef))
  (bind _returnvalue (Transformation|GetActorLocation _waterexclusionref))
  (bind _water_cube (Variables|Default|GetWater_Cube))
  (bind _relativelocation (Variables|/Script/Engine.SceneComponent|GetRelativeLocation _water_cube))
  ; Init
  (Variables|BuildingActor|SetbIsInvulnerable true)
  (Variables|Actor|SetbCanBeDamaged false)
  ; Set Water Start/End Locations
  (Variables|Default|SetWaterStartLocation _relativelocation)
  (Variables|Default|SetWaterEndLocation (+ _relativelocation (Variables|Default|GetWater%20Cube%20Movement)))
  ; Set Water Exclusion Volume Start/End Locations
  (Variables|Default|SetExclusionVolumeStartLoc _returnvalue)
  (Variables|Default|SetExclusionVolumeEndLoc (+ _returnvalue (Variables|Default|GetWater%20Exclusion%20Volume%20Movement))))

(event Custom|Activate%20Water%20Platform
  (bind _arrow (Variables|Default|GetArrow))
  (bind _returnvalue (Transformation|GetWorldLocation _arrow))
  ; Activate Event
  (Math|Integer|IncrementInt (Variables|Default|GetSwitches%20Active)
    (:%20%20
      (if (>= (Variables|Default|GetSwitches%20Active) (Variables|Default|GetSwitches%20Required))
        (Networking|SwitchHasAuthority
          (:Authority
            (Components|Activation|Activate (Variables|Default|GetWaterfall%20FX) false)
            (bind _output_get (Variables|Default|SetWaterPlatformIsActive true))
            (if _output_get
              (bind (_direction _watermovement) (Timeline|WaterCube_Timeline Play :NewTime (Variables|Default|GetTravel%20Time))
                (:Update
                  (Transformation|SetRelativeLocation (Variables|Default|GetWater_Cube) (Math|Vector|Lerp(Vector) (Variables|Default|GetWaterStartLocation) (Variables|Default|GetWaterEndLocation) _watermovement) false false)
                  (Actor|DestroyActor (Variables|Default|GetWaterbody%20%28River%29%20Ref))
                  (Transformation|SetActorRelativeLocation (Variables|Default|GetWaterExclusionRef) (Math|Vector|Lerp(Vector) (Variables|Default|GetExclusionVolumeStartLoc) (Variables|Default|GetExclusionVolumeEndLoc) _watermovement) false false))
                (:Finished
                  (Audio|Components|Audio|StopDelayed (Variables|Default|GetDrain%20Audio) 1.0)
                  (Components|Activation|Deactivate (Variables|Default|GetWaterfall%20FX))))
              (else
                (Timeline|WaterCube_Timeline Reverse))))
          (:Remote
            (Audio|PlaySoundatLocation "/Game/Sounds/Foley_Loot/Containers/Pirate_Booty/Pirate_Booty_Beacon_Activate_Cue.Pirate_Booty_Beacon_Activate_Cue" _returnvalue)
            (Audio|Components|Audio|Play (Variables|Default|GetDrain%20Audio) 0.0)
            (Niagara|SpawnSystematLocation 0 _returnvalue)))))))

(event Custom|Deactivate%20Water%20Platform
  ; Deeactivate Water Platform
  (Networking|SwitchHasAuthority
    (:Authority
      (bind _output_get (Variables|Default|SetWaterPlatformIsActive false))
      (if _output_get
        ; Activate Event
        (Timeline|WaterCube_Timeline Play)
        (else
          (Utilities|FlowControl|Delay (Variables|Default|GetWater%20Raise%20Delay))
          (Development|PrintString "Raise Water" true true "(R=0.000000,G=0.660000,B=1.000000,A=1.000000)" 1e+01)
          (Timeline|WaterCube_Timeline Reverse))))
    (:Remote
      (Audio|Components|Audio|Play (Variables|Default|GetFill%20Audio) 0.0)
      (Niagara|SpawnSystematLocation 0 (Transformation|GetWorldLocation (Variables|Default|GetArrow))))))

(fn OnRep_IsinTrigger)

(fn OnRep_WaterPlatformIsActive)

(fn ConstructionScript)
