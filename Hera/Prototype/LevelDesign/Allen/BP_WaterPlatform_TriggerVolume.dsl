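; Prototype invulnerable water platform trigger that lowers a water cube while a player is inside and raises it after exit.
; On the server, player-pawn overlaps play WaterCube_Timeline, or reverse it after a short delay on exit.
; The timeline lerps the cube down and, behind a Delay in its update, moves the referenced water exclusion actor upward.
; Remote machines only spawn looping water sounds; debug prints remain and OnRep_IsinTrigger is empty.
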(event EventBeginPlay
  (bind _waterexclusionref (Variables|Default|GetWaterExclusionRef))
  (bind _returnvalue (Transformation|GetActorLocation _waterexclusionref))
  (bind _water_cube (Variables|Default|GetWater_Cube))
  (bind _relativelocation (Variables|/Script/Engine.SceneComponent|GetRelativeLocation _water_cube))
  ; Init
  (Variables|BuildingActor|SetbIsInvulnerable true)
  (Variables|Actor|SetbCanBeDamaged false)
  ; Set Water Start/End Locations
  (Variables|Default|SetWaterStartLocation _relativelocation)
  (Variables|Default|SetWaterEndLocation (+ _relativelocation "0.000000,0.000000,-384.000000"))
  ; Set Water Exclusion Volume Start/End Locations
  (Variables|Default|SetExclusionVolumeStartLoc _returnvalue)
  (Variables|Default|SetExclusionVolumeEndLoc (+ _returnvalue "0.000000,0.000000,384.000000")))

(event OnComponentBeginOverlap(TriggerVolume)
  (params (OverlappedComponent :type Primitive%20Component%20Object%20Reference) (OtherActor :type Actor%20Object%20Reference) (OtherComp :type Primitive%20Component%20Object%20Reference) (OtherBodyIndex :type Integer) (bFromSweep :type Boolean) (SweepResult :type Hit%20Result%20Structure%20%28by%20ref%29))
  ; Trigger Events
  (Networking|SwitchHasAuthority
    (:Authority
      (Utilities|Casting|CastToFortPlayerPawnAthena OtherActor) ; this checks, on the server, if a player touched this, not a bot or wildlife
      (bind _output_get (Variables|Default|SetIsinTrigger true))
      (if _output_get
        (bind (_direction _watermovement) (Timeline|WaterCube_Timeline Play)
          (:Update
            (Development|PrintString "Lower Water" true true "(R=0.000000,G=0.660000,B=1.000000,A=1.000000)" 1e+01)
            (Transformation|SetRelativeLocation (Variables|Default|GetWater_Cube) (Math|Vector|Lerp(Vector) (Variables|Default|GetWaterStartLocation) (Variables|Default|GetWaterEndLocation) _watermovement) false false)
            (Development|PrintString "Move Water" true true "(R=1.000000,G=0.000000,B=0.795774,A=1.000000)" 1e+01)
            (Utilities|FlowControl|Delay 0.8)
            (Transformation|SetActorRelativeLocation (Variables|Default|GetWaterExclusionRef) (Math|Vector|Lerp(Vector) (Variables|Default|GetExclusionVolumeStartLoc) (Variables|Default|GetExclusionVolumeEndLoc) _watermovement) false false)
            (Development|PrintString "Move Exclusion Volume" true true "(R=1.000000,G=0.000000,B=0.754528,A=1.000000)" 1e+01)))
        (else
          (Timeline|WaterCube_Timeline Reverse))))
    (:Remote
      (Audio|SpawnSoundatLocation "/Game/Sounds/Asteria/Environment_Actors/WaterPipeOverflow/Spot_Water_PipeOverflow_DrainOpening_Large_Loop.Spot_Water_PipeOverflow_DrainOpening_Large_Loop" (Transformation|GetWorldLocation (Variables|Default|GetWater_Cube))))))

(event OnComponentEndOverlap(TriggerVolume)
  (params (OverlappedComponent :type Primitive%20Component%20Object%20Reference) (OtherActor :type Actor%20Object%20Reference) (OtherComp :type Primitive%20Component%20Object%20Reference) (OtherBodyIndex :type Integer))
  ; Exit Events
  (Networking|SwitchHasAuthority
    (:Authority
      (Utilities|Casting|CastToFortPlayerPawnAthena OtherActor) ; this checks, on the server, if a player touched this, not a bot or wildlife
      (bind _output_get (Variables|Default|SetIsinTrigger false))
      (if _output_get
        ; Trigger Events
        (Timeline|WaterCube_Timeline Play)
        (else
          (Utilities|FlowControl|Delay 2.0)
          (Development|PrintString "Raise Water" true true "(R=0.000000,G=0.660000,B=1.000000,A=1.000000)" 1e+01)
          (Timeline|WaterCube_Timeline Reverse))))
    (:Remote
      (Audio|SpawnSoundatLocation "/AmbientAudioContent/SoundCues/SpottedLoops/Spot_Water_PipeOverflow_MainFlow_Large_Loop_Cue.Spot_Water_PipeOverflow_MainFlow_Large_Loop_Cue" (Transformation|GetWorldLocation (Variables|Default|GetWater_Cube))))))

(fn OnRep_IsinTrigger)

(fn ConstructionScript)
