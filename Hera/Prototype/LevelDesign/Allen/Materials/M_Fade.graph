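# M_Fade is a masked Default Lit material that computes a dithered vertical fade across an object's bounds.
# It takes the Z component of bounding-box 0-1 UVW via a dot product and inverts it.
# The Fade Amount scalar parameter is subtracted to shift the gradient, and a power curve shapes the falloff.
# The result drives the DitherTemporalAA threshold, approximating soft transparency under masked blending.
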Material /Hera/Prototype/LevelDesign/Allen/Materials/M_Fade
graph {
 properties {bAutomaticallySetUsageInEditor=false, blendMode=BLEND_Masked, usedShadingModels=Default Lit}
 nodes [
  Constant() {r=0.3}
  Constant() {r=0.8}
  Constant() {r=1}
  n1 = OneMinus(None=DotProduct(
    A=MaterialFunctionCall() {materialFunction=&BoundingBoxBased_0-1_UVW}.RGB
    B=Constant3Vector() {constant=<rgba 0 0 1 0>}
   ))
  n2 = ScalarParameter() {defaultValue=1, parameterName=Fade Amount, sliderMax=1, sliderMin=-1}
  MaterialFunctionCall(Alpha Threshold=Power(Base=Subtract(A=@n1, B=@n2)) {constExponent=1.5}) {materialFunction=&DitherTemporalAA}
 ]
}
