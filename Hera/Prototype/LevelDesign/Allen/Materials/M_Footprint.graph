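# M_Footprint is a prototype translucent deferred decal material built around a footprint icon from the UI icon library.
# It adds the icon texture's alpha to a constant bright blue color and multiplies the sum to boost intensity.
# Because the alpha is added instead of used as a multiplier, the blue remains outside the footprint shape and the print shifts brighter.
# The graph shows no material output connection for this result.
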Material /Hera/Prototype/LevelDesign/Allen/Materials/M_Footprint
graph {
 properties {bAutomaticallySetUsageInEditor=false, blendMode=BLEND_Translucent, materialDomain=MD_DeferredDecal, usedShadingModels=Default Lit}
 nodes [Multiply(A=Add(
    A=TextureSample() {texture=/Game/Athena/UI/InGame/Creative/PropertyEditor/Icon_Picker/Textures_V2/128x128/T_UI_IconLibrary_Footprint_128}.A
    B=Constant3Vector() {constant=<rgba 0 0.4977 1 1>}
   )) {constB=6}]
}
