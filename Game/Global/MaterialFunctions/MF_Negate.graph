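# MF_Negate is a library material function, exposed as "Negate" under Math, that flips the sign of a scalar or vector.
# It has four parallel input-to-output lanes, f1 through f4, for scalar, two-, three- and four-component values.
# Each lane multiplies its input by negative one, and every input can fall back on its preview value when left unconnected.
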MaterialFunction /Game/Global/MaterialFunctions/MF_Negate
graph {
 properties {bExposeToLibrary=true, description="Negates (x * -1.0) a scalar or vector input", userExposedCaption=Negate, libraryCategoriesText [Math]}
 nodes [
  n1 = FunctionInput() {bUsePreviewValueAsDefault=true, inputName=f1, inputType=FunctionInput_Scalar, previewValue=<1 1 1 1>}
  FunctionOutput(None=Multiply(A=@n1) {constB=-1}) {outputName=f1}
  n2 = FunctionInput() {bUsePreviewValueAsDefault=true, inputName=f2, inputType=FunctionInput_Vector2, sortPriority=1, previewValue=<1 1 1 1>}
  FunctionOutput(None=Multiply(A=@n2) {constB=-1}) {outputName=f2, sortPriority=1}
  FunctionOutput(None=Multiply(A=FunctionInput() {bUsePreviewValueAsDefault=true, inputName=f3, sortPriority=2, previewValue=<1 1 1 1>}) {constB=-1}) {outputName=f3, sortPriority=2}
  n3 = FunctionInput() {bUsePreviewValueAsDefault=true, inputName=f4, inputType=FunctionInput_Vector4, sortPriority=3, previewValue=<1 1 1 1>}
  FunctionOutput(None=Multiply(A=@n3) {constB=-1}) {outputName=f4, sortPriority=3}
 ]
}
