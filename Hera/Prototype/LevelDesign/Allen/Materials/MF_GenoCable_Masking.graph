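# Material function for Geno cable masking that outputs elapsed time since a StartTime input, limited to a selected index.
# It subtracts StartTime from a Time input, which defaults to the material Time node.
# A saturated floor of StartTime zeroes the result while StartTime is below 1, acting as a branchless activation gate.
# MF_FloatEqual compares IndexMask with Index and multiplies the result to mask by index.
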MaterialFunction /Hera/Prototype/LevelDesign/Allen/Materials/MF_GenoCable_Masking
graph {nodes [
  n1 = FunctionInput() {inputName=StartTime, inputType=FunctionInput_Scalar}
  n2 = Subtract(A=FunctionInput(Preview=Time()) {bUsePreviewValueAsDefault=true, inputName=Time, inputType=FunctionInput_Scalar, sortPriority=4}, B=@n1)
  n3 = Saturate(None=Floor(None=@n1))
  n4 = FunctionInput() {bUsePreviewValueAsDefault=true, inputName=IndexMask, inputType=FunctionInput_Scalar, sortPriority=3}
  n5 = FunctionInput() {bUsePreviewValueAsDefault=true, inputName=Index, inputType=FunctionInput_Scalar, sortPriority=2}
  FunctionOutput(None=Multiply(A=Multiply(A=@n2, B=@n3), B=MaterialFunctionCall!(A=@n4, B=@n5) {materialFunction=/Game/Global/MaterialFunctions/MF_FloatEqual}.Return))
 ]}
