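# Prototype Nanite material for GenoCable bundles: dark carbon-fiber cables with emissive pulses traveling along their length.
# Vertex colors supply bundle ID, cable ID and distance along the cable.
# Eight custom primitive data start times feed MF_GenoCable_Masking, selected by rounded bundle ID and offset by cable ID, to drive data pulses.
# These add to a repeating time-scrolled sine glow, with both tinted and attenuated by inverse Fresnel.
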Material /Hera/Prototype/LevelDesign/Allen/Materials/M_GenoCable
graph {
 properties {bAutomaticallySetUsageInEditor=false, bUseMaterialAttributes=true, bUsedWithNanite=true, usedShadingModels=Default Lit}
 nodes [
  n1 = NamedRerouteUsage()
  Add(A=@n1, B=ScalarParameter() {defaultValue=1, parameterName=Len})
  n2 = VertexColor() {channelNames {a="", b=CableDistance, g=CableID, r=CableBundleID}}
  n3 = Multiply(A=Add(A=@n2.CableDistance, B=@n2.CableDistance), B=NamedRerouteUsage())
  n4 = NamedRerouteUsage()
  Add(A=@n3, B=@n4)
  n5 = VertexColor() {channelNames {a="", b=CableDistance, g=CableID, r=CableBundleID}}
  n6 = Round(None=Multiply(
    A=VertexColor() {channelNames {a="", b=CableDistance, g=CableID, r=CableBundleID}}.CableBundleID
    B=ScalarParameter() {defaultValue=1, parameterName=NumCables}
   ) {constB=2})
  n7 = Add(
   A=MaterialFunctionCall!(StartTime=ScalarParameter() {bUseCustomPrimitiveData=true, parameterName=Data.Cable0.Time}, Index=Constant(), IndexMask=@n6, Time) {materialFunction=&/MF_GenoCable_Masking}.Result
   B=MaterialFunctionCall(
    StartTime=ScalarParameter() {bUseCustomPrimitiveData=true, parameterName=Data.Cable1.Time, primitiveDataIndex=1}
    Index=Constant() {r=1}
    IndexMask=@n6
   ) {materialFunction=&MF_GenoCable_Masking}.Result
  )
  n8 = MaterialFunctionCall(
   StartTime=ScalarParameter() {bUseCustomPrimitiveData=true, parameterName=Data.Cable2.Time, primitiveDataIndex=2}
   Index=Constant() {r=2}
   IndexMask=@n6
  ) {materialFunction=&MF_GenoCable_Masking}
  n9 = ScalarParameter() {bUseCustomPrimitiveData=true, parameterName=Data.Cable3.Time, primitiveDataIndex=3}
  n10 = Constant() {r=3}
  n11 = Add(
   A=Add(
    A=Add(A=@n7, B=@n8.Result)
    B=MaterialFunctionCall(StartTime=@n9, Index=@n10, IndexMask=@n6) {materialFunction=&MF_GenoCable_Masking}.Result
   )
   B=MaterialFunctionCall(
    StartTime=ScalarParameter() {bUseCustomPrimitiveData=true, parameterName=Data.Cable4.Time, primitiveDataIndex=4}
    Index=Constant() {r=4}
    IndexMask=@n6
   ) {materialFunction=&MF_GenoCable_Masking}.Result
  )
  n12 = MaterialFunctionCall(
   StartTime=ScalarParameter() {bUseCustomPrimitiveData=true, parameterName=Data.Cable5.Time, primitiveDataIndex=5}
   Index=Constant() {r=5}
   IndexMask=@n6
  ) {materialFunction=&MF_GenoCable_Masking}
  n13 = ScalarParameter() {bUseCustomPrimitiveData=true, parameterName=Data.Cable6.Time, primitiveDataIndex=6}
  n14 = Constant() {r=6}
  n15 = Add(
   A=Add(
    A=Add(A=@n11, B=@n12.Result)
    B=MaterialFunctionCall(StartTime=@n13, Index=@n14, IndexMask=@n6) {materialFunction=&MF_GenoCable_Masking}.Result
   )
   B=MaterialFunctionCall(
    StartTime=ScalarParameter() {bUseCustomPrimitiveData=true, parameterName=Data.Cable7.Time, primitiveDataIndex=7}
    Index=Constant() {r=7}
    IndexMask=@n6
   ) {materialFunction=&MF_GenoCable_Masking}.Result
  )
  n16 = Multiply(A=@n5.CableID, B=ScalarParameter() {desc=In centimeters per second, parameterName=CableTimeOffset, sliderMax=1})
  n17 = Subtract(A=Multiply(A=@n5.CableDistance, B=NamedRerouteUsage()), B=Multiply(A=Add(A=@n15, B=@n16)) {constB=5000})
  Add(A=@n17, B=Multiply() {constA=2000})
  NamedRerouteDeclaration(None=@n1) {name=ActualLength, nodeColor=<rgba 1 0.5144 0 1>}
  NamedRerouteDeclaration(None=ScalarParameter() {defaultValue=6254, desc=In centimeters, parameterName=CurveMaxLengthWS}) {name=CurveLength, nodeColor=<rgba 1 0.5144 0 1>}
  NamedRerouteDeclaration(None=Divide(A=NamedRerouteUsage())) {name=StartOffset, nodeColor=<rgba 1 0.0938 0 1>}
  n18 = ScalarParameter() {defaultValue=200, desc=In centimeters, parameterName=SpreadDistanceWS}
  NamedRerouteDeclaration(None=Max(
    A=ScalarParameter() {defaultValue=2000, desc=In centimeters per second, parameterName=SpreadFrequencyWS}
    B=NamedRerouteDeclaration(None=@n18) {name=Spread, nodeColor=<rgba 0 1 0.2956 1>}
   )) {name=SpreadFrequency, nodeColor=<rgba 0 1 0.2956 1>}
  NamedRerouteUsage()
  ScalarParameter() {defaultValue=50, desc=In centimeters per second, parameterName=SpreadSpeedWS}
  ScalarParameter() {defaultValue=1, parameterName=TimeOverride}
  n19 = Time()
  n20 = ScalarParameter() {defaultValue=50, desc=In centimeters per second, parameterName=SpreadSpeedWS}
  n21 = NamedRerouteUsage()
  n22 = Divide(A=Subtract(A=@n3, B=Multiply(A=@n19, B=@n20)), B=@n21)
  n23 = Divide(
   A=Multiply(A=MaterialFunctionCall(Value=@n22) {materialFunction=&LinearSine}.Linear Sine, B=@n21)
   B=NamedRerouteUsage()
  )
  n24 = SmoothStep(Value=OneMinus(None=Saturate(None=@n23)))
  n25 = Multiply(
   A=Multiply(A=Power(Base=@n24) {constExponent=8}, B=ScalarParameter() {defaultValue=0.3, parameterName=Emissive.Intensity}) {constB=0.3}
   B=VectorParameter() {parameterName=Emissive.Tint, defaultValue=<rgba 1 1 1 1>}.RGB
  )
  n26 = Divide(A=MaterialFunctionCall!(f1=@n17, f2, f3, f4) {materialFunction=/Game/Global/MaterialFunctions/MF_Negate}.f1) {constB=2000}
  n27 = Abs(None=Subtract(A=@n26) {constB=2})
  n28 = Power(Base=OneMinus(None=Saturate(None=@n27))) {constExponent=8}
  n29 = ScalarParameter() {defaultValue=50, parameterName=DataPulse.Emissive.Scale}
  n30 = Saturate(None=Subtract(A=@n26))
  n31 = Power(Base=SmoothStep(Value=@n30)) {constExponent=8}
  n32 = ScalarParameter() {defaultValue=1, parameterName=DataEmissive.Intensity}
  n33 = Multiply(
   A=Add(A=Multiply(A=@n28, B=@n29), B=Multiply(A=@n31, B=@n32) {constB=0.3})
   B=VectorParameter() {parameterName=DataEmissive.Tint, defaultValue=<rgba 1 1 1 1>}.RGB
  )
  n34 = Fresnel(Normal=PixelNormalWS()) {baseReflectFraction=0, exponent=0.5}
  n35 = TextureCoordinate()
  n36 = AppendVector(A=ScalarParameter() {defaultValue=0.15, parameterName=Tile.X}, B=ScalarParameter() {defaultValue=0.3, parameterName=Tile.Y})
  SetMaterialAttributes!(
   MaterialAttributes
   Base Color=Constant3Vector() {constant=<rgba 0.0365 0.0365 0.0365 0>}
   Emissive Color=Multiply(A=Add(A=@n25, B=@n33), B=OneMinus(None=@n34))
   Normal=TextureSample(UVs=Multiply(A=@n35, B=@n36)) {samplerType=SAMPLERTYPE_Normal, texture=/Game/Weapons/WeaponSkins/Wraps/Textures/T_Wrap_CarbonFiber_N}.RGB
  ) {attributeSetTypes [#1, #2, #3]}
  n37 = Constant3Vector() {constant=<rgba 1 0 0 0>}
  SetMaterialAttributes!(MaterialAttributes, Base Color=If(A=@n30, "A > B"=@n37, "A == B"=@n37, "A < B"=@n30) {constB=1}) {attributeSetTypes [#1]}
  Subtract(A=@n1, B=NamedRerouteUsage())
  Subtract(A=@n3, B=@n4)
  Time()
 ]
}
