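# Exposed Utility material function "Slerp" that spherically interpolates between unit vectors A and B by scalar Alpha.
# Clamps their dot product, takes the arccosine times Alpha, and divides by 2π for the period-based Sine and Cosine nodes.
# These weight A and a normalized perpendicular built from B, which yields NaN when A equals B.
# Separate Fast and Precise outputs differ only in using ArccosineFast or Arccosine.
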MaterialFunction /Game/Global/MaterialFunctions/MF_Slerp
graph {
 properties {
  bExposeToLibrary=true
  description="Returns the spherical lerp between unit vectors A and B. If A = B, returns NaN"
  userExposedCaption=Slerp
  libraryCategoriesText [Utility]
 }
 nodes [
  n1 = Clamp(None=DotProduct(A=NamedRerouteUsage(), B=NamedRerouteUsage())) {minDefault=-1}
  n2 = Divide(
   A=Multiply(A=NamedRerouteUsage(), B=ArccosineFast!(None=@n1))
   B=MaterialFunctionCall(Multiplier=Constant() {r=2}) {materialFunction=&Pi}.Result
  )
  n3 = Sine(None=@n2)
  n4 = NamedRerouteUsage()
  n5 = Normalize(VectorInput=Subtract(A=NamedRerouteUsage(), B=Multiply(A=@n1, B=@n4)))
  n6 = Cosine(None=@n2)
  n7 = NamedRerouteUsage()
  FunctionOutput(None=Add(A=Multiply(A=@n3, B=@n5), B=Multiply(A=@n6, B=@n7))) {description=Uses ACosFast, outputName=Fast}
  n8 = Divide(
   A=Multiply(A=NamedRerouteUsage(), B=Arccosine!(None=@n1))
   B=MaterialFunctionCall(Multiplier=Constant() {r=2}) {materialFunction=&Pi}.Result
  )
  n9 = Sine(None=@n8)
  n10 = Cosine(None=@n8)
  n11 = NamedRerouteUsage()
  FunctionOutput(None=Add(A=Multiply(A=@n9, B=@n5), B=Multiply(A=@n10, B=@n11))) {outputName=Precise, sortPriority=1}
  n12 = FunctionInput() {bUsePreviewValueAsDefault=true, description=Expects normalized unit vector, inputName=A, previewValue=<1 0 0 1>}
  NamedRerouteDeclaration(None=@n12) {name=P0, nodeColor=<rgba 1 0.2109 0 1>}
  n13 = FunctionInput() {bUsePreviewValueAsDefault=true, description=Expects normalized unit vector, inputName=B, sortPriority=1, previewValue=<0 1 0 1>}
  NamedRerouteDeclaration(None=@n13) {name=P1, nodeColor=<rgba 1 0.2109 0 1>}
  n14 = FunctionInput() {
   bUsePreviewValueAsDefault=true
   description="Blend between A (0.0) and B (1.0)"
   inputName=Alpha
   inputType=FunctionInput_Scalar
   sortPriority=2
   previewValue=<0.5 0 0 1>
  }
  NamedRerouteDeclaration(None=@n14) {name=T, nodeColor=<rgba 0.4297 1 0 1>}
 ]
}
