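# Material function SphericalDistance, exposed in the Utility library, returns the angular distance between two unit vectors A and B.
# It takes the dot product of the inputs and clamps it from -1 to 1 before taking the arccosine.
# Two outputs trade cost for accuracy: Fast uses the approximate fast arccosine and Precise uses the full arccosine.
# Inputs are expected to be already normalized, as the function does no normalization itself.
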MaterialFunction /Game/Global/MaterialFunctions/MF_SphericalDistance
graph {
 properties {
  bExposeToLibrary=true
  description="Returns the spherical distance between two unit vectors A and B. Expects normalized vectors. "
  userExposedCaption=SphericalDistance
  libraryCategoriesText [Utility]
 }
 nodes [
  n1 = FunctionInput() {bUsePreviewValueAsDefault=true, description=Expects normalized unit vector, inputName=A, previewValue=<1 0 0 1>}
  n2 = FunctionInput() {bUsePreviewValueAsDefault=true, description=Expects normalized unit vector, inputName=B, sortPriority=1, previewValue=<0 1 0 1>}
  n3 = Clamp(None=DotProduct(A=@n1, B=@n2)) {minDefault=-1}
  FunctionOutput(None=ArccosineFast!(None=@n3)) {description=Uses ACosFast, outputName=Fast}
  FunctionOutput(None=Arccosine!(None=@n3)) {outputName=Precise, sortPriority=1}
 ]
}
