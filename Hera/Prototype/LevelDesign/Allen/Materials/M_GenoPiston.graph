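# Prototype material for the Geno piston, with vertex-animation math that turns the mesh about a pivot and slides pistons.
# Rotation slerps from previous to current look-at target in custom primitive data, eased over angular distance, staggered by vertex green, with matching rotated normals.
# Piston offsets follow directions encoded in vertex color, blending a looping sine pulse with a timed contraction.
# Vertex red selects between rotated and default values; surface shading uses simple parameters.
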Material /Hera/Prototype/LevelDesign/Allen/Materials/M_GenoPiston
graph {
 properties {bAutomaticallySetUsageInEditor=false, bTangentSpaceNormal=false, usedShadingModels=Default Lit}
 nodes [
  n1 = Time()
  n2 = DeltaTime()
  Add(A=@n1, B=@n2)
  Constant() {r=1}
  n3 = VertexColor()
  n4 = VertexNormalWS()
  If(A=@n3.R, "A > B"=NamedRerouteUsage(), "A == B"=@n4, "A < B"=@n4)
  n5 = Constant()
  If(A=@n3.R, "A > B"=NamedRerouteUsage(), "A == B"=@n5, "A < B"=@n5)
  n6 = Time()
  n7 = Constant() {r=6}
  n8 = Multiply(A=@n7) {constB=0.5}
  n9 = VertexColor()
  n10 = MaterialFunctionCall(Value=Add(A=Divide(A=@n6, B=@n8) {constB=3}, B=Length(None=@n9) {is=class MaterialExpressionLength})) {materialFunction=&LinearSine}
  n11 = Saturate(None=Subtract(A=Multiply(A=@n10.Linear Sine, B=@n7) {constB=6}, B=Add(A=@n8)) {constB=3})
  n12 = Subtract(A=Time(), B=ScalarParameter() {bUseCustomPrimitiveData=true, parameterName=PistonContractionTime, primitiveDataIndex=19})
  n13 = Constant() {r=20}
  n14 = Multiply(A=@n13) {constB=0.5}
  n15 = MaterialFunctionCall(Value=Saturate(None=Divide(A=@n12, B=@n14) {constB=3})) {materialFunction=&LinearSine}
  n16 = Saturate(None=Subtract(A=Multiply(A=@n15.Linear Sine, B=@n13) {constB=6}, B=Add(A=@n14)) {constB=3})
  Max(A=@n11, B=@n16)
  n17 = Subtract(A=@n9) {constB=0.5}
  Multiply(
   A=Multiply(A=Normalize(VectorInput=@n17), B=LinearInterpolate(A=@n11, Alpha=@n16))
   B=ScalarParameter() {defaultValue=-200, parameterName=PistonMotionDistance}
  ) {constB=-200}
  n18 = Normalize(VectorInput=NamedRerouteUsage())
  NamedRerouteDeclaration(None=@n18) {name=ForwardBasis, nodeColor=<rgba 1 0 0.75 1>}
  n19 = Normalize(VectorInput=CrossProduct(A=@n18, B=Constant3Vector() {constant=<rgba 0 0 1 0>}))
  NamedRerouteDeclaration(None=@n19) {name=RightBasis, nodeColor=<rgba 1 0 0.75 1>}
  n20 = NamedRerouteUsage()
  n21 = NamedRerouteUsage()
  n22 = NamedRerouteUsage()
  n23 = NamedRerouteUsage()
  NamedRerouteDeclaration(None=Multiply(A=MaterialFunctionCall(VectorToTransform=@n20, BasisX=@n21, BasisY=@n22, BasisZ=@n23) {materialFunction=&Transform3x3Matrix}.Transformed Vector) {constB=-1}) {name=RotatedNormalOrientation, nodeColor=<rgba 1 0 0.9913 1>}
  n24 = VectorParameter() {
   bUseCustomPrimitiveData=true
   parameterName=LookAtTarget
   primitiveDataIndex=3
   channelNames {a="", b=Z, g=Y, r=X}
   defaultValue=<rgba 1 0 0 0>
  }
  n25 = Constant3Vector() {constant=<rgba 0 1 0 0>}
  NamedRerouteDeclaration(None=MaterialFunctionCall(Vector=@n24.RGB, Default=@n25) {materialFunction=&SafeNormalize}.Result) {name=LookAtTarget, nodeColor=<rgba 0.8516 0 1 1>}
  n26 = VectorParameter() {
   bUseCustomPrimitiveData=true
   parameterName=PreviousLookAtTarget
   primitiveDataIndex=6
   channelNames {a="", b=Z, g=Y, r=X}
   defaultValue=<rgba 0 0 0 1>
  }
  NamedRerouteDeclaration(None=MaterialFunctionCall(Vector=@n26.RGB, Default=@n25) {materialFunction=&SafeNormalize}.Result) {name=PreviousLookAtTarget, nodeColor=<rgba 0.8516 0 1 1>}
  NamedRerouteDeclaration(None=Normalize(VectorInput=CrossProduct(A=@n18, B=@n19))) {name=UpBasis, nodeColor=<rgba 1 0 0.75 1>}
  n27 = VectorParameter() {bUseCustomPrimitiveData=true, parameterName=PivotLocation, channelNames {a="", b=Z, g=Y, r=X}, defaultValue=<rgba 0 0 6452.1499 0>}
  n28 = Add(
   A=MaterialFunctionCall() {materialFunction=&ComponentPivotLocation}.Result
   B=@n27.RGB
  )
  n29 = WorldPosition() {worldPositionShaderOffset=WPT_ExcludeAllShaderOffsets}
  NamedRerouteDeclaration(None=Subtract(A=@n28, B=@n29.XYZ)) {name=LocalOffset, nodeColor=<rgba 1 0.3063 0 1>}
  NamedRerouteDeclaration(None=Constant3Vector() {constant=<rgba 0 0 1 0>}) {name="[0,0,1]", nodeColor=<rgba 0 0 1 1>}
  NamedRerouteDeclaration(None=Constant3Vector() {constant=<rgba 1 0 0 0>}) {name="[1,0,0]", nodeColor=<rgba 1 0 0 1>}
  NamedRerouteDeclaration(None=Constant3Vector() {constant=<rgba 0 1 0 0>}) {name="[0,1,0]", nodeColor=<rgba 0 1 0 1>}
  NamedRerouteDeclaration(None=VertexNormalWS()) {name=NormalsWS, nodeColor=<rgba 0.5 0.5 1 1>}
  n30 = NamedRerouteUsage()
  n31 = NamedRerouteUsage()
  n32 = NamedRerouteUsage()
  n33 = Subtract(A=Time(), B=ScalarParameter() {bUseCustomPrimitiveData=true, parameterName=CurrentTime, primitiveDataIndex=9})
  n34 = Divide(A=MaterialFunctionCall!(A=NamedRerouteUsage(), B=NamedRerouteUsage()) {materialFunction=/Game/Global/MaterialFunctions/MF_SphericalDistance}.Fast) {constB=1}
  n35 = OneMinus(None=VertexColor().G)
  n36 = Saturate(None=Subtract(A=Divide(A=@n33, B=@n34) {constB=0.25}, B=Multiply(A=@n35) {constB=0.1}))
  n37 = Power(Base=Multiply(A=OneMinus(None=@n36)) {constB=2}) {constExponent=4}
  n38 = Multiply(A=@n36) {constB=2}
  n39 = If(A=@n36, "A > B"=OneMinus(None=Multiply(A=@n37) {constB=0.5}), "A < B"=Multiply(A=Power(Base=@n38) {constExponent=4}) {constB=0.5}) {constB=0.5}
  NamedRerouteDeclaration(None=If(
    A=DotProduct(A=@n30, B=@n31)
    "A > B"=@n32
    "A == B"=@n32
    "A < B"=MaterialFunctionCall!(A=@n30, B=@n31, Alpha=@n39) {materialFunction=/Game/Global/MaterialFunctions/MF_Slerp}.Fast
   ) {constB=1}) {name=InterpTarget, nodeColor=<rgba 0.5 0 1 1>}
  n40 = NamedRerouteUsage()
  n41 = NamedRerouteUsage()
  n42 = NamedRerouteUsage()
  n43 = NamedRerouteUsage()
  n44 = NamedRerouteUsage()
  n45 = NamedRerouteUsage()
  n46 = MaterialFunctionCall(
   VectorToTransform=MaterialFunctionCall(R=DotProduct(A=@n40, B=@n41), G=DotProduct(A=@n42, B=@n43), B=DotProduct(A=@n44, B=@n45)) {materialFunction=&AppendMany}.RGB
   BasisX=NamedRerouteUsage()
   BasisY=NamedRerouteUsage()
   BasisZ=NamedRerouteUsage()
  ) {materialFunction=&Transform3x3Matrix}
  NamedRerouteDeclaration(None=Subtract(A=Add(A=@n28, B=@n46.Transformed Vector), B=@n29.XYZ)) {name=RotatedVertexPostion, nodeColor=<rgba 0 0.5078 1 1>}
  PreviousFrameSwitch()
  PreviousFrameSwitch(Current Frame=@n1, Previous Frame=@n1)
  ScalarParameter() {parameterName=Metallic}
  ScalarParameter() {defaultValue=0.5, parameterName=Roughness}
  n47 = Constant() {r=1.7016}
  Subtract(A=Multiply(A=Power(Base=@n36) {constExponent=3}, B=Add(A=@n47)), B=Multiply(A=Power(Base=@n36), B=@n47))
  Subtract(A=@n1, B=Multiply(A=@n2) {constB=10})
  VectorParameter() {parameterName=Color, defaultValue=<rgba 0.0938 0.0938 0.0938 1>}
  VectorParameter() {parameterName=EmissiveColor, defaultValue=<rgba 0 0 0 1>}
 ]
}
