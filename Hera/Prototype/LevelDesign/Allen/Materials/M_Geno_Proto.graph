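# M_Geno_Proto is a prototype Nanite material that animates look-at eye rotation through World Position Offset and rotated normals, behind the UseEyeRotation switch.
# Custom primitive data supplies pivot, timing, focal distances and current and previous targets for channels A and B, blended with eased slerps.
# Vertex red staggers timing and selects parts pivoted by focal distance, while green masks the deformation.
# Shading uses simple PBR parameters with Fresnel-dimmed, exposure-compensated emissive.
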Material /Hera/Prototype/LevelDesign/Allen/Materials/M_Geno_Proto
graph {
 properties {
  bAutomaticallySetUsageInEditor=false
  bTangentSpaceNormal=false
  bUseMaterialAttributes=true
  bUsedWithNanite=true
  usedShadingModels=Default Lit
 }
 nodes [
  LocalPosition()
  MaterialFunctionCall!(f1, f2, f3, f4) {materialFunction=/Game/Global/MaterialFunctions/MF_Negate}
  n1 = Normalize(VectorInput=NamedRerouteUsage())
  NamedRerouteDeclaration(None=@n1) {name=ForwardBasis, nodeColor=<rgba 1 0 0.75 1>}
  n2 = Normalize(VectorInput=CrossProduct(A=@n1, B=Constant3Vector() {constant=<rgba 0 0 1 0>}))
  NamedRerouteDeclaration(None=@n2) {name=RightBasis, nodeColor=<rgba 1 0 0.75 1>}
  n3 = Multiply(A=NamedRerouteUsage(), B=NamedRerouteUsage())
  n4 = NamedRerouteUsage()
  n5 = NamedRerouteUsage()
  n6 = NamedRerouteUsage()
  NamedRerouteDeclaration(None=Multiply(A=MaterialFunctionCall(VectorToTransform=@n3, BasisX=@n4, BasisY=@n5, BasisZ=@n6) {materialFunction=&Transform3x3Matrix}.Transformed Vector) {constB=-1}) {name=RotatedNormalOrientation, nodeColor=<rgba 1 0 0.9913 1>}
  n7 = VectorParameter() {
   bUseCustomPrimitiveData=true
   parameterName=LookAtTargetA
   primitiveDataIndex=3
   channelNames {a="", b=Z, g=Y, r=X}
   defaultValue=<rgba 1 0 0 0>
  }
  n8 = Constant3Vector() {constant=<rgba 0 1 0 0>}
  NamedRerouteDeclaration(None=MaterialFunctionCall(Vector=@n7.RGB, Default=@n8) {materialFunction=&SafeNormalize}.Result) {name=LookAtTargetA, nodeColor=<rgba 0.8516 0 1 1>}
  n9 = VectorParameter() {
   bUseCustomPrimitiveData=true
   parameterName=PreviousLookAtTargetA
   primitiveDataIndex=6
   channelNames {a="", b=Z, g=Y, r=X}
   defaultValue=<rgba 0 0 0 1>
  }
  NamedRerouteDeclaration(None=MaterialFunctionCall(Vector=@n9.RGB, Default=@n8) {materialFunction=&SafeNormalize}.Result) {name=PreviousLookAtTargetA, nodeColor=<rgba 0.8516 0 1 1>}
  n10 = Divide(
   A=Subtract(A=Time(), B=ScalarParameter() {bUseCustomPrimitiveData=true, parameterName=CurrentTimeA, primitiveDataIndex=9})
   B=ScalarParameter() {bUseCustomPrimitiveData=true, parameterName=CurrentDurationA, primitiveDataIndex=10}
  ) {constB=0.25}
  n11 = VertexColor()
  n12 = Multiply(A=OneMinus(None=Ceil(None=@n11.R))) {constB=0.05}
  NamedRerouteDeclaration(None=Saturate(None=Subtract(A=@n10, B=@n12))) {name=TimelineA, nodeColor=<rgba 1 0.2109 0 1>}
  n13 = VectorParameter() {
   bUseCustomPrimitiveData=true
   parameterName=LookAtTargetB
   primitiveDataIndex=11
   channelNames {a="", b=Z, g=Y, r=X}
   defaultValue=<rgba 1 0 0 0>
  }
  n14 = Constant3Vector() {constant=<rgba 0 -1 0 0>}
  NamedRerouteDeclaration(None=MaterialFunctionCall(Vector=@n13.RGB, Default=@n14) {materialFunction=&SafeNormalize}.Result) {name=LookAtTargetB, nodeColor=<rgba 0.8516 0 1 1>}
  n15 = VectorParameter() {
   bUseCustomPrimitiveData=true
   parameterName=PreviousLookAtTargetB
   primitiveDataIndex=14
   channelNames {a="", b=Z, g=Y, r=X}
   defaultValue=<rgba 0 0 0 1>
  }
  NamedRerouteDeclaration(None=MaterialFunctionCall(Vector=@n15.RGB, Default=@n14) {materialFunction=&SafeNormalize}.Result) {name=PreviousLookAtTargetB, nodeColor=<rgba 0.8516 0 1 1>}
  n16 = Divide(
   A=Subtract(A=Time(), B=ScalarParameter() {bUseCustomPrimitiveData=true, parameterName=CurrentTimeB, primitiveDataIndex=17})
   B=ScalarParameter() {bUseCustomPrimitiveData=true, parameterName=CurrentDurationB, primitiveDataIndex=18}
  ) {constB=0.25}
  NamedRerouteDeclaration(None=Saturate(None=Subtract(A=@n16, B=@n12))) {name=TimelineB, nodeColor=<rgba 1 0.2109 0 1>}
  n17 = TransformPosition(None=WorldPosition() {worldPositionShaderOffset=WPT_ExcludeAllShaderOffsets}.XYZ) {transformSourceType=TRANSFORMPOSSOURCE_World}
  n18 = ComponentMask(None=Floor(None=Saturate(None=@n17))) {g=true}
  n19 = Multiply(A=Subtract(A=NamedRerouteDeclaration(None=@n18) {name=DirectionalMask, nodeColor=<rgba 1 0.6328 0 1>}) {constB=0.5}) {constB=2}
  NamedRerouteDeclaration(None=MaterialFunctionCall(R=@n19, G=@n19, B=Constant() {r=1}) {materialFunction=&AppendMany}.RGB) {name=FlipStartDirection, nodeColor=<rgba 0.6172 0 1 1>}
  NamedRerouteDeclaration(None=Normalize(VectorInput=CrossProduct(A=@n1, B=@n2))) {name=UpBasis, nodeColor=<rgba 1 0 0.75 1>}
  n20 = VectorParameter() {bUseCustomPrimitiveData=true, parameterName=PivotLocation, channelNames {a="", b=Z, g=Y, r=X}, defaultValue=<rgba 0 0 6452.1499 0>}
  n21 = Add(
   A=MaterialFunctionCall() {materialFunction=&ComponentPivotLocation}.Result
   B=@n20.RGB
  )
  n22 = ScalarParameter() {bUseCustomPrimitiveData=true, parameterName=PreviousFocalDistanceA, primitiveDataIndex=21}
  n23 = ScalarParameter() {bUseCustomPrimitiveData=true, parameterName=FocalDistanceA, primitiveDataIndex=20}
  n24 = NamedRerouteUsage()
  n25 = ScalarParameter() {bUseCustomPrimitiveData=true, parameterName=PreviousFocalDistanceB, primitiveDataIndex=23}
  n26 = ScalarParameter() {bUseCustomPrimitiveData=true, parameterName=FocalDistanceB, primitiveDataIndex=22}
  n27 = NamedRerouteUsage()
  n28 = InverseLinearInterpolate(
   A=ScalarParameter() {defaultValue=10000, parameterName=FocalDistanceMin}
   B=ScalarParameter() {defaultValue=80000, parameterName=FocalDistanceMax}
   Value=LinearInterpolate(A=LinearInterpolate(A=@n22, B=@n23, Alpha=@n24), B=LinearInterpolate(A=@n25, B=@n26, Alpha=@n27), Alpha=NamedRerouteUsage())
  ) {constA=10000, constB=80000}
  n29 = Constant3Vector() {constant=<rgba 1 0 0 0>}
  n30 = VertexColor()
  n31 = LinearInterpolate(A=Constant3Vector() {constant=<rgba -0.8667 0 0.5 0>}, B=Constant3Vector() {constant=<rgba -0.8667 0 -0.5 0>}, Alpha=Floor(None=@n30.R))
  n32 = Round(None=@n30.R)
  n33 = Subtract(A=NamedRerouteUsage()) {constB=0.5}
  n34 = WorldPosition()
  n35 = MaterialFunctionCall!(
   RotationAngle=LinearInterpolate(Alpha=Saturate(None=@n28)) {constA=0.04, constB=-0.01}
   NormalizedRotationAxis=Multiply(A=LinearInterpolate(A=@n29, B=@n31, Alpha=@n32), B=Multiply(A=@n33) {constB=2})
   PivotPoint=@n21
   Position=@n34.XYZ
   Normal
   TangentU
   TangentV
  ) {materialFunction=/Game/Characters/Player/Male/Medium/Bodies/M_MED_Inferno/FX/Mat/RotateVertexAboutAxis}
  n36 = Ceil(None=@n30.R)
  n37 = LinearInterpolate(B=@n35.WPO, Alpha=@n36)
  NamedRerouteDeclaration(None=Subtract(A=@n21, B=Add(A=@n37, B=@n34.XYZ))) {name=LocalOffset, nodeColor=<rgba 1 0.3063 0 1>}
  NamedRerouteDeclaration(None=Constant3Vector() {constant=<rgba 0 0 1 0>}) {name="[0,0,1]", nodeColor=<rgba 0 0 1 1>}
  NamedRerouteDeclaration(None=Constant3Vector() {constant=<rgba 1 0 0 0>}) {name="[1,0,0]", nodeColor=<rgba 1 0 0 1>}
  NamedRerouteDeclaration(None=Constant3Vector() {constant=<rgba 0 1 0 0>}) {name="[0,1,0]", nodeColor=<rgba 0 1 0 1>}
  NamedRerouteDeclaration(None=LinearInterpolate(A=VertexNormalWS(), B=@n35.Normal, Alpha=@n36)) {name=NormalsWS, nodeColor=<rgba 0.5 0.5 1 1>}
  n38 = NamedRerouteUsage()
  n39 = NamedRerouteUsage()
  n40 = DotProduct(A=@n38, B=@n39)
  n41 = NamedRerouteUsage()
  n42 = NamedRerouteUsage()
  n43 = OneMinus(None=@n42)
  n44 = Multiply(A=Power(Base=Multiply(A=@n43) {constB=2}) {constExponent=4}) {constB=0.5}
  n45 = Power(Base=Multiply(A=@n42) {constB=2}) {constExponent=4}
  n46 = MaterialFunctionCall!(A=@n38, B=@n39, Alpha=If(A=NamedRerouteUsage(), "A > B"=OneMinus(None=@n44), "A < B"=Multiply(A=@n45) {constB=0.5}) {constB=0.5}) {materialFunction=/Game/Global/MaterialFunctions/MF_Slerp}
  n47 = NamedRerouteUsage()
  n48 = NamedRerouteUsage()
  n49 = DotProduct(A=@n47, B=@n48)
  n50 = NamedRerouteUsage()
  n51 = NamedRerouteUsage()
  n52 = OneMinus(None=@n51)
  n53 = Multiply(A=Power(Base=Multiply(A=@n52) {constB=2}) {constExponent=4}) {constB=0.5}
  n54 = Power(Base=Multiply(A=@n51) {constB=2}) {constExponent=4}
  n55 = MaterialFunctionCall(A=@n47, B=@n48, Alpha=If(A=NamedRerouteUsage(), "A > B"=OneMinus(None=@n53), "A < B"=Multiply(A=@n54) {constB=0.5}) {constB=0.5}) {materialFunction=&MF_Slerp}
  NamedRerouteDeclaration(None=LinearInterpolate(
    A=If(A=@n40, "A > B"=@n41, "A == B"=@n41, "A < B"=@n46.Fast) {constB=1}
    B=If(A=@n49, "A > B"=@n50, "A == B"=@n50, "A < B"=@n55.Fast) {constB=1}
    Alpha=NamedRerouteUsage()
   )) {name=InterpTarget, nodeColor=<rgba 0.5 0 1 1>}
  n56 = DotProduct(A=NamedRerouteUsage(), B=NamedRerouteUsage())
  n57 = DotProduct(A=NamedRerouteUsage(), B=NamedRerouteUsage())
  n58 = DotProduct(A=NamedRerouteUsage(), B=NamedRerouteUsage())
  n59 = MaterialFunctionCall(
   VectorToTransform=Multiply(
    A=MaterialFunctionCall(R=@n56, G=@n57, B=@n58) {materialFunction=&AppendMany}.RGB
    B=NamedRerouteUsage()
   )
   BasisX=NamedRerouteUsage()
   BasisY=NamedRerouteUsage()
   BasisZ=NamedRerouteUsage()
  ) {materialFunction=&Transform3x3Matrix}
  NamedRerouteDeclaration(None=Subtract(A=Add(A=@n21, B=@n59.Transformed Vector), B=WorldPosition() {worldPositionShaderOffset=WPT_ExcludeAllShaderOffsets}.XYZ)) {name=RotatedVertexPostion, nodeColor=<rgba 0 0.5078 1 1>}
  SetMaterialAttributes!(MaterialAttributes, Base Color=NamedRerouteUsage()) {attributeSetTypes [#1]}
  n60 = EyeAdaptationInverse(LightValueInput=VectorParameter() {parameterName=EmissiveColor, defaultValue=<rgba 0 0 0 1>}.RGB)
  n61 = SetMaterialAttributes!(
   MaterialAttributes
   Base Color=VectorParameter() {parameterName=Color, defaultValue=<rgba 0.0938 0.0938 0.0938 1>}.RGB
   Metallic=ScalarParameter() {parameterName=Metallic}
   Roughness=ScalarParameter() {defaultValue=0.5, parameterName=Roughness}
   Emissive Color=LinearInterpolate(A=@n60.EyeAdaptationInverse, B=Multiply(A=@n60.EyeAdaptationInverse) {constB=0.125}, Alpha=Fresnel() {baseReflectFraction=0, exponent=1})
  ) {attributeSetTypes [
    #1
    #2
    #3
    #4
   ]}
  n62 = VertexColor()
  n63 = NamedRerouteUsage()
  n64 = VertexNormalWS()
  n65 = NamedRerouteUsage()
  n66 = Constant()
  StaticSwitch(
   True=SetMaterialAttributes!(
    MaterialAttributes=@n61
    Normal=If(A=@n62.G, "A > B"=@n63, "A == B"=@n64, "A < B"=@n64)
    World Position Offset=If(A=@n62.G, "A > B"=@n65, "A == B"=@n66, "A < B"=@n66)
   ) {attributeSetTypes [#5, #6]}
   False=@n61
   Value=StaticBoolParameter() {parameterName=UseEyeRotation}
  )
  n67 = NamedRerouteUsage()
  n68 = Constant() {r=1.7016}
  Subtract(A=Multiply(A=Power(Base=@n67) {constExponent=3}, B=Add(A=@n68)), B=Multiply(A=Power(Base=@n67), B=@n68))
  VertexColor()
  VertexNormalWS()
 ]
}
