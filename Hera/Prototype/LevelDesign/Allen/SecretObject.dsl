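; Prototype secret-object building Blueprint whose material reacts to nearby players and damage.
; Construction picks a Hidden or Secret parent material by object type and applies TextureData textures to a dynamic instance.
; Server-side ProximityVisibility overlaps set PlayerClose, whose OnRep writes signed game time to "Visibility.Time", the sign encoding proximity.
; Damage cycles through indexed "Hit" vector parameters, packing impact location and time into each for shader-driven effects, alongside debug prints.
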(event EventBeginPlay)

(event OnComponentBeginOverlap(ProximityVisibility)
  (params (OverlappedComponent :type Primitive%20Component%20Object%20Reference) (OtherActor :type Actor%20Object%20Reference) (OtherComp :type Primitive%20Component%20Object%20Reference) (OtherBodyIndex :type Integer) (bFromSweep :type Boolean) (SweepResult :type Hit%20Result%20Structure%20%28by%20ref%29))
  (Networking|SwitchHasAuthority
    (:Authority
      (Utilities|Casting|CastToPlayerPawn_Athena OtherActor)
      (Development|PrintString "QuitStepping on Meh")
      (Variables|Private|SetPlayerClose true))))

(event OnComponentEndOverlap(ProximityVisibility)
  (params (OverlappedComponent :type Primitive%20Component%20Object%20Reference) (OtherActor :type Actor%20Object%20Reference) (OtherComp :type Primitive%20Component%20Object%20Reference) (OtherBodyIndex :type Integer))
  (Networking|SwitchHasAuthority
    (:Authority
      (Utilities|Casting|CastToPlayerPawn_Athena OtherActor)
      (if (== (Utilities|Array|Length (Collision|GetOverlappingActors (Variables|Default|GetProximityVisibility) "/Game/Athena/PlayerPawn_Athena.PlayerPawn_Athena_C" :!class /Script/Engine.PrimitiveComponent)) 0)
        (Development|PrintString "Whew that feels better")
        (Variables|Private|SetPlayerClose false)))))

(event EventOnDamagePlayEffects
  (params (Damage :type Float%20%28single-precision%29) (DamageTags :type Gameplay%20Tag%20Container%20Structure%20%28by%20ref%29) (Momentum :type Vector :default "0, 0, 0") (HitInfo :type Hit%20Result%20Structure%20%28by%20ref%29) (InstigatedBy :type Fort%20Pawn%20Object%20Reference) (DamageCauser :type Actor%20Object%20Reference) (EffectContext :type Gameplay%20Effect%20Context%20Handle%20Structure))
  (bind (_bblockinghit _binitialoverlap _time _distance _location _impactpoint _normal _impactnormal _physmat _hitactor _hitcomponent) (Collision|BreakHitResult HitInfo))
  (Networking|CanExecuteCosmeticEvents
    (:True
      (Development|PrintString (Utilities|String|Append "GAH! Ahm HIT! At " (Utilities|String|ToString(Vector) _impactpoint)))
      (Variables|Private|SetLastHitComponent _hitcomponent)
      (Variables|Private|SetLastImpactPoint _impactpoint)
      (Variables|Private|SetCurrentHit (Math|Integer|%25%28Integer%29 (+ (Variables|Private|GetCurrentHit) 1) (Variables|Private|GetNumSupportedHits)))
      (Default|SetHit (Variables|Default|GetStaticMeshComponent) (Variables|Private|GetCurrentHit) (Variables|Private|GetLastImpactPoint)))))

(fn CreateAndSetDynamicMaterialInstance
  (params (Primitive :type Primitive%20Component%20Object%20Reference))
  (bind _primitive (Variables|GetPrimitive))
  (for _index (range (Rendering|Material|GetNumMaterials Primitive))
    (Utilities|Casting|CastToMaterialInstanceDynamic (Rendering|Material|GetMaterial _primitive _index)
      (:CastFailed
        (Rendering|Material|CreateDynamicMaterialInstance (Variables|GetPrimitive) _index (Rendering|Material|GetMaterial _primitive _index) (Utilities|String|StringToName (Utilities|String|Append "MID_" (Utilities|GetDisplayName (Rendering|Material|GetMaterial _primitive _index)))) :!class /Script/Engine.PrimitiveComponent)))))

(fn OnRep_PlayerClose
  (Networking|CanExecuteCosmeticEvents
    (:True
      (Development|PrintString "Am I Close?")
      (Default|SetAllScalarParameters (Variables|Default|GetStaticMeshComponent) "Visibility.Time" (* (Utilities|Time|GetGameTimeinSeconds) (Math|Float|SelectFloat 1.0 -1.0 (Variables|Private|GetPlayerClose)))))))

(fn SetAllScalarParameters
  (params (Primitive :type Primitive%20Component%20Object%20Reference) (ParameterName :type Name) (Value :type Float%20%28double-precision%29))
  (for _index (range (Rendering|Material|GetNumMaterials Primitive))
    (bind _asmaterial_instance_dynamic (Utilities|Casting|CastToMaterialInstanceDynamic (Rendering|Material|GetMaterial (Variables|GetPrimitive) _index)))
    (Rendering|Material|SetScalarParameterValue _asmaterial_instance_dynamic (Variables|GetParameterName) (Variables|GetValue) :!class /Script/Engine.MaterialInstanceDynamic)))

(fn SetAllVectorParameters
  (params (Primitive :type Primitive%20Component%20Object%20Reference) (ParameterName :type Name) (Value :type Linear%20Color%20Structure))
  (for _index (range (Rendering|Material|GetNumMaterials Primitive))
    (bind _asmaterial_instance_dynamic (Utilities|Casting|CastToMaterialInstanceDynamic (Rendering|Material|GetMaterial (Variables|GetPrimitive) _index)))
    (Rendering|Material|SetVectorParameterValue _asmaterial_instance_dynamic (Variables|GetParameterName) (Variables|GetValue) :!class /Script/Engine.MaterialInstanceDynamic)))

(fn SetHit
  (params (Primitive :type Primitive%20Component%20Object%20Reference) (HitIndex :type Integer) (HitLocation :type Vector))
  (Default|SetAllVectorParameters Primitive (Utilities|String|StringToName (Utilities|String|Append "Hit" (Utilities|String|ToString(Integer) HitIndex))) (Math|Color|MakeColor (.x HitLocation) (.y HitLocation) (.z HitLocation) (Utilities|Time|GetGameTimeinSeconds))))

(fn SetParentMaterial
  (switch Utilities|FlowControl|Switch|SwitchonE_SecretObjectType (Variables|SecretObject|GetSecretObjectType)
    (:NewEnumerator0
      (LocalVariables|SetParentMaterial (Variables|Private|GetHiddenMaterial)))
    (:NewEnumerator1
      (LocalVariables|SetParentMaterial (Variables|Private|GetSecretMaterial))))
  (bind _asmaterial_instance_dynamic (Utilities|Casting|CastToMaterialInstanceDynamic (Rendering|Material|GetMaterial (Variables|Default|GetStaticMeshComponent) 0))
    (:then
      (if (== (Variables|/Script/Engine.MaterialInstance|GetParent _asmaterial_instance_dynamic) (LocalVariables|GetParentMaterial))
        _
        (else
          _)))
    (:CastFailed
      _))
  (Rendering|Material|CreateDynamicMaterialInstance (Variables|Default|GetStaticMeshComponent) 0 (LocalVariables|GetParentMaterial) "None" :!class /Script/Engine.PrimitiveComponent))

(fn SetTDs
  (bind _texturedata (Variables|Default|GetTextureData))
  (bind _output (Utilities|Array|Get(acopy) _texturedata))
  (bind _texturedata_2 (Variables|Default|GetTextureData))
  (bind _diffuse (Variables|/Script/FortniteGame.BuildingTextureData|GetDiffuse _output))
  (bind _specular (Variables|/Script/FortniteGame.BuildingTextureData|GetSpecular _output))
  (bind _normal (Variables|/Script/FortniteGame.BuildingTextureData|GetNormal _output))
  (bind _asmaterial_instance_dynamic (Utilities|Casting|CastToMaterialInstanceDynamic (Rendering|Material|GetMaterial (Variables|Default|GetStaticMeshComponent) 0)))
  (Utilities|IsValid _output
    (:Is%20Valid
      (Utilities|IsValid _diffuse
        (:Is%20Valid
          (Rendering|Material|SetTextureParameterValue _asmaterial_instance_dynamic "Diffuse" _diffuse))
        (:Is%20Not%20Valid
          _))
      (Utilities|IsValid _specular
        (:Is%20Valid
          (Rendering|Material|SetTextureParameterValue _asmaterial_instance_dynamic "SpecularMasks" _specular)
          (Utilities|IsValid _normal
            (:Is%20Valid
              (Rendering|Material|SetTextureParameterValue _asmaterial_instance_dynamic "Normals" _normal))))))
    (:Is%20Not%20Valid
      _))
  (for _index (range 1 4)
    (Utilities|IsValid (Utilities|Array|Get(acopy) _texturedata_2 _index)
      (:Is%20Valid
        (Utilities|IsValid (Variables|/Script/FortniteGame.BuildingTextureData|GetDiffuse (Utilities|Array|Get(acopy) _texturedata_2 _index))
          (:Is%20Valid
            (Rendering|Material|SetTextureParameterValue _asmaterial_instance_dynamic (Utilities|String|StringToName (Utilities|String|Append "Diffuse_Texture_" (Utilities|String|ToString(Integer) (+ _index 1)))) (Variables|/Script/FortniteGame.BuildingTextureData|GetDiffuse (Utilities|Array|Get(acopy) _texturedata_2 _index))))
          (:Is%20Not%20Valid
            _))
        (Utilities|IsValid (Variables|/Script/FortniteGame.BuildingTextureData|GetSpecular (Utilities|Array|Get(acopy) _texturedata_2 _index))
          (:Is%20Valid
            (Rendering|Material|SetTextureParameterValue _asmaterial_instance_dynamic (Utilities|String|StringToName (Utilities|String|Append "SpecularMasks_" (Utilities|String|ToString(Integer) (+ _index 1)))) (Variables|/Script/FortniteGame.BuildingTextureData|GetSpecular (Utilities|Array|Get(acopy) _texturedata_2 _index)))
            (Utilities|IsValid (Variables|/Script/FortniteGame.BuildingTextureData|GetNormal (Utilities|Array|Get(acopy) _texturedata_2 _index))
              (:Is%20Valid
                (Rendering|Material|SetTextureParameterValue _asmaterial_instance_dynamic (Utilities|String|StringToName (Utilities|String|Append "Normals_Texture_" (Utilities|String|ToString(Integer) (+ _index 1)))) (Variables|/Script/FortniteGame.BuildingTextureData|GetNormal (Utilities|Array|Get(acopy) _texturedata_2 _index)))))))))))

(fn ConstructionScript
  (Utilities|FlowControl|Sequence
    (:then_0
      (if (!= (Utilities|Array|Length (Variables|Default|GetTextureData)) 4)
        (Utilities|Array|Resize (Variables|Default|GetTextureData) 4)))
    (:then_1
      (CallFunction|SetParentMaterial)
      (CallFunction|SetTDs))))
