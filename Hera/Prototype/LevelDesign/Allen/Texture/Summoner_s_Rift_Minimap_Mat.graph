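# Prototype material built around a Summoner's Rift minimap texture, kept in a level design texture folder.
# The graph holds a single texture sample of that minimap under Default Lit shading, with no connection to a material output shown.
# Automatic usage setting in the editor is turned off.
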Material /Hera/Prototype/LevelDesign/Allen/Texture/Summoner_s_Rift_Minimap_Mat
graph {
 properties {bAutomaticallySetUsageInEditor=false, usedShadingModels=Default Lit}
 nodes [TextureSample() {texture=&/Summoner_s_Rift_Minimap}]
}
